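; LandscapeWaterInfo is a Blueprint that applies a configured flood water level to the ocean.
; Its custom Set Flood Water Level event gets the Water Subsystem and calls SetOceanFloodHeight with the Flood Water Level variable.
; BeginPlay and the construction script are empty, so the height changes only when that event is invoked.
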(event EventBeginPlay)

(event Custom|Set%20Flood%20Water%20Level
  (Water|SetOceanFloodHeight (WorldSubsystems|GetWaterSubsystem) (Variables|Default|GetFlood%20Water%20Level)))

(fn ConstructionScript)
