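; Actor component that swaps its owner's water materials to mobile versions on mobile, Switch, PC performance mode, or when Fort.Water.UseMobileWater is set.
; Type flags pick the target component, async-loading materials with optional overrides, including river transitions and WaterZone far mesh; waterfalls are hidden.
; BeginPlay and a DataDrivenCVar delegate run duplicated copies of this logic, with no revert path.
; DebugMobileWater_DontSave applies it unconditionally.
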(event Custom|DebugMobileWater_DontSave
  (bind _returnvalue (Components|GetOwner))
  (bind _returnvalue_2 (Actor|GetComponentbyClass _returnvalue "/Script/Water.WaterBodyLakeComponent"))
  (bind _returnvalue_3 (Components|GetOwner))
  (bind _returnvalue_4 (Actor|GetComponentbyClass _returnvalue_3 "/Script/Water.WaterBodyCustomComponent"))
  (bind _returnvalue_5 (Components|GetOwner))
  (bind _returnvalue_6 (Components|GetOwner))
  (bind _returnvalue_7 (Components|GetOwner))
  (bind _returnvalue_8 (Actor|GetComponentbyClass _returnvalue_7 "/Script/Water.WaterBodyLakeComponent"))
  (bind _returnvalue_9 (Components|GetOwner))
  (bind _returnvalue_10 (Components|GetOwner))
  (bind _returnvalue_11 (Actor|GetComponentbyClass _returnvalue_10 "/Script/Water.WaterBodyCustomComponent"))
  (bind _returnvalue_12 (Actor|GetComponentbyClass _returnvalue_6 "/Script/Engine.StaticMeshComponent"))
  (bind _returnvalue_13 (Actor|GetComponentbyClass _returnvalue_9 "/Script/Water.WaterBodyOceanComponent"))
  (bind _returnvalue_14 (Actor|GetComponentbyClass _returnvalue_5 "/Script/Water.WaterBodyRiverComponent"))
  (Utilities|FlowControl|Sequence
    (:then_0
      (if (Variables|Default|GetIsStaticMesh)
        (Utilities|IsValid _returnvalue_12
          (:Is%20Valid
            (bind _asmaterial_interface (Utilities|CustomAsyncLoad (Variables|Default|GetStaticMeshMaterial) (Variables|Default|GetCustom_OverrideMaterial))
              (:Out
                (Rendering|Material|SetMaterial _returnvalue_12 0 _asmaterial_interface)))))))
    (:then_1
      (if (Variables|Default|GetIsOceanExtendOuter)
        (bind _aswater_zone (Utilities|Casting|CastToWaterZone (Components|GetOwner)))
        (bind _asmaterial_interface_2 (Utilities|CustomAsyncLoad (Variables|Default|GetStaticMeshMaterial) (Variables|Default|GetCustom_OverrideMaterial))
          (:Out
            (Rendering|SetFarMeshMaterial _aswater_zone _asmaterial_interface_2)))))
    (:then_2
      (if (Variables|Default|GetIsOceanExtend)
        (Utilities|IsValid _returnvalue_2
          (:Is%20Valid
            (bind _asmaterial_interface_3 (Utilities|CustomAsyncLoad (Variables|Default|GetOceanMaterial_Extend) (Variables|Default|GetCustom_OverrideMaterial))
              (:Out
                (Rendering|SetWaterMaterial _returnvalue_2 _asmaterial_interface_3)))))))
    (:then_3
      (if (Variables|Default|GetIsLake)
        (Utilities|IsValid _returnvalue_8
          (:Is%20Valid
            (bind _asmaterial_interface_4 (Utilities|CustomAsyncLoad (Variables|Default|GetLakeMaterial) (Variables|Default|GetCustom_OverrideMaterial))
              (:Out
                (Rendering|SetWaterMaterial _returnvalue_8 _asmaterial_interface_4)))))))
    (:then_4
      (if (Variables|Default|GetIsOcean)
        (Utilities|IsValid _returnvalue_13
          (:Is%20Valid
            (bind _asmaterial_interface_5 (Utilities|CustomAsyncLoad (Variables|Default|GetOceanMaterial) (Variables|Default|GetCustom_OverrideMaterial))
              (:Out
                (Rendering|SetWaterMaterial _returnvalue_13 _asmaterial_interface_5)))))))
    (:then_5
      (if (Variables|Default|GetIsRiver)
        (Utilities|IsValid _returnvalue_14
          (:Is%20Valid
            (Utilities|FlowControl|Sequence
              (:then_0
                (bind _asmaterial_interface_6 (Utilities|CustomAsyncLoad (Variables|Default|GetRiverMaterial) (Variables|Default|GetCustom_OverrideMaterial))
                  (:Out
                    (Rendering|SetWaterMaterial _returnvalue_14 _asmaterial_interface_6))))
              (:then_1
                (bind _asmaterial_interface_7 (Utilities|CustomAsyncLoad (Variables|Default|GetLakeTransitionMaterial) (Variables|Default|GetTo_Lake_Custom_OverrideMaterial))
                  (:Out
                    (bind _asmaterial_interface_8 (Utilities|CustomAsyncLoad (Variables|Default|GetOceanTransitionMaterial) (Variables|Default|GetTo_Ocean_Custom_OverrideMaterial))
                      (:Out
                        (Rendering|SetLakeandOceanTransitionMaterials _returnvalue_14 _asmaterial_interface_7 _asmaterial_interface_8)))))))))))
    (:then_6
      (if (Variables|Default|GetIsPool)
        (Utilities|IsValid _returnvalue_11
          (:Is%20Valid
            (bind _asmaterial_interface_9 (Utilities|CustomAsyncLoad (Variables|Default|GetPoolMaterial) (Variables|Default|GetCustom_OverrideMaterial))
              (:Out
                (Rendering|SetWaterMaterial _returnvalue_11 _asmaterial_interface_9)))))))
    (:then_7
      (if (Variables|Default|GetIsWaterfall)
        (Utilities|IsValid _returnvalue_4
          (:Is%20Valid
            (Rendering|SetVisibility _returnvalue_4 false false)))))))

(event EventBeginPlay
  (bind _returnvalue (Components|GetOwner))
  (bind _returnvalue_2 (Actor|GetComponentbyClass _returnvalue "/Script/Water.WaterBodyLakeComponent"))
  (bind _returnvalue_3 (Components|GetOwner))
  (bind _returnvalue_4 (Actor|GetComponentbyClass _returnvalue_3 "/Script/Water.WaterBodyCustomComponent"))
  (bind _returnvalue_5 (Components|GetOwner))
  (bind _returnvalue_6 (Components|GetOwner))
  (bind _returnvalue_7 (Components|GetOwner))
  (bind _returnvalue_8 (Actor|GetComponentbyClass _returnvalue_7 "/Script/Water.WaterBodyLakeComponent"))
  (bind _returnvalue_9 (Components|GetOwner))
  (bind _returnvalue_10 (Components|GetOwner))
  (bind _returnvalue_11 (Actor|GetComponentbyClass _returnvalue_10 "/Script/Water.WaterBodyCustomComponent"))
  (bind _returnvalue_12 (Actor|GetComponentbyClass _returnvalue_6 "/Script/Engine.StaticMeshComponent"))
  (bind _returnvalue_13 (Actor|GetComponentbyClass _returnvalue_9 "/Script/Water.WaterBodyOceanComponent"))
  (bind _returnvalue_14 (Actor|GetComponentbyClass _returnvalue_5 "/Script/Water.WaterBodyRiverComponent"))
  (DataDrivenCVar|BindEventtoOnDataDrivenCVarDelegate (EngineSubsystems|GetDataDrivenCVarEngineSubsystem) (delegate Custom|ChangeWaterType))
  (bind _returnvalue_15 (Development|GetConsoleVariableBoolValue "Fort.Water.UseMobileWater"))
  (if (Math|Boolean|ORBoolean _returnvalue_15 (Class|PlatformHelpersLibrary|IsPcPerfMode) (TouchPlatform|IsNintendoSwitchorNintendoSwitchPreview) (TouchPlatform|IsMobileGame))
    (Utilities|FlowControl|Sequence
      (:then_0
        (if (Variables|Default|GetIsStaticMesh)
          (Utilities|IsValid _returnvalue_12
            (:Is%20Valid
              (bind _asmaterial_interface (Utilities|CustomAsyncLoad (Variables|Default|GetStaticMeshMaterial) (Variables|Default|GetCustom_OverrideMaterial))
                (:Out
                  (Rendering|Material|SetMaterial _returnvalue_12 0 _asmaterial_interface)))))))
      (:then_1
        (if (Variables|Default|GetIsOceanExtendOuter)
          (bind _aswater_zone (Utilities|Casting|CastToWaterZone (Components|GetOwner)))
          (bind _asmaterial_interface_2 (Utilities|CustomAsyncLoad (Variables|Default|GetStaticMeshMaterial) (Variables|Default|GetCustom_OverrideMaterial))
            (:Out
              (Rendering|SetFarMeshMaterial _aswater_zone _asmaterial_interface_2)))))
      (:then_2
        (if (Variables|Default|GetIsOceanExtend)
          (Utilities|IsValid _returnvalue_2
            (:Is%20Valid
              (bind _asmaterial_interface_3 (Utilities|CustomAsyncLoad (Variables|Default|GetOceanMaterial_Extend) (Variables|Default|GetCustom_OverrideMaterial))
                (:Out
                  (Rendering|SetWaterMaterial _returnvalue_2 _asmaterial_interface_3)))))))
      (:then_3
        (if (Variables|Default|GetIsLake)
          (Utilities|IsValid _returnvalue_8
            (:Is%20Valid
              (bind _asmaterial_interface_4 (Utilities|CustomAsyncLoad (Variables|Default|GetLakeMaterial) (Variables|Default|GetCustom_OverrideMaterial))
                (:Out
                  (Rendering|SetWaterMaterial _returnvalue_8 _asmaterial_interface_4)))))))
      (:then_4
        (if (Variables|Default|GetIsOcean)
          (Utilities|IsValid _returnvalue_13
            (:Is%20Valid
              (bind _asmaterial_interface_5 (Utilities|CustomAsyncLoad (Variables|Default|GetOceanMaterial) (Variables|Default|GetCustom_OverrideMaterial))
                (:Out
                  (Rendering|SetWaterMaterial _returnvalue_13 _asmaterial_interface_5)))))))
      (:then_5
        (if (Variables|Default|GetIsRiver)
          (Utilities|IsValid _returnvalue_14
            (:Is%20Valid
              (Utilities|FlowControl|Sequence
                (:then_0
                  (bind _asmaterial_interface_6 (Utilities|CustomAsyncLoad (Variables|Default|GetRiverMaterial) (Variables|Default|GetCustom_OverrideMaterial))
                    (:Out
                      (Rendering|SetWaterMaterial _returnvalue_14 _asmaterial_interface_6))))
                (:then_1
                  (bind _asmaterial_interface_7 (Utilities|CustomAsyncLoad (Variables|Default|GetLakeTransitionMaterial) (Variables|Default|GetTo_Lake_Custom_OverrideMaterial))
                    (:Out
                      (bind _asmaterial_interface_8 (Utilities|CustomAsyncLoad (Variables|Default|GetOceanTransitionMaterial) (Variables|Default|GetTo_Ocean_Custom_OverrideMaterial))
                        (:Out
                          (Rendering|SetLakeandOceanTransitionMaterials _returnvalue_14 _asmaterial_interface_7 _asmaterial_interface_8)))))))))))
      (:then_6
        (if (Variables|Default|GetIsPool)
          (Utilities|IsValid _returnvalue_11
            (:Is%20Valid
              (bind _asmaterial_interface_9 (Utilities|CustomAsyncLoad (Variables|Default|GetPoolMaterial) (Variables|Default|GetCustom_OverrideMaterial))
                (:Out
                  (Rendering|SetWaterMaterial _returnvalue_11 _asmaterial_interface_9)))))))
      (:then_7
        (if (Variables|Default|GetIsWaterfall)
          (Utilities|IsValid _returnvalue_4
            (:Is%20Valid
              (Rendering|SetVisibility _returnvalue_4 false false))))))))

(event Custom|ChangeWaterType
  (params (CVarName :type String))
  (bind _returnvalue (Components|GetOwner))
  (bind _returnvalue_2 (Actor|GetComponentbyClass _returnvalue "/Script/Water.WaterBodyLakeComponent"))
  (bind _returnvalue_3 (Components|GetOwner))
  (bind _returnvalue_4 (Actor|GetComponentbyClass _returnvalue_3 "/Script/Water.WaterBodyCustomComponent"))
  (bind _returnvalue_5 (Components|GetOwner))
  (bind _returnvalue_6 (Components|GetOwner))
  (bind _returnvalue_7 (Components|GetOwner))
  (bind _returnvalue_8 (Actor|GetComponentbyClass _returnvalue_7 "/Script/Water.WaterBodyLakeComponent"))
  (bind _returnvalue_9 (Components|GetOwner))
  (bind _returnvalue_10 (Components|GetOwner))
  (bind _returnvalue_11 (Actor|GetComponentbyClass _returnvalue_10 "/Script/Water.WaterBodyCustomComponent"))
  (bind _returnvalue_12 (Actor|GetComponentbyClass _returnvalue_6 "/Script/Engine.StaticMeshComponent"))
  (bind _returnvalue_13 (Actor|GetComponentbyClass _returnvalue_9 "/Script/Water.WaterBodyOceanComponent"))
  (bind _returnvalue_14 (Actor|GetComponentbyClass _returnvalue_5 "/Script/Water.WaterBodyRiverComponent"))
  (if (Utilities|String|EqualExactly(String) CVarName "Fort.Water.UseMobileWater")
    (bind _returnvalue_15 (Development|GetConsoleVariableBoolValue "Fort.Water.UseMobileWater"))
    (if (Math|Boolean|ORBoolean _returnvalue_15 (Class|PlatformHelpersLibrary|IsPcPerfMode) (TouchPlatform|IsNintendoSwitchorNintendoSwitchPreview) (TouchPlatform|IsMobileGame))
      (Utilities|FlowControl|Sequence
        (:then_0
          (if (Variables|Default|GetIsStaticMesh)
            (Utilities|IsValid _returnvalue_12
              (:Is%20Valid
                (bind _asmaterial_interface (Utilities|CustomAsyncLoad (Variables|Default|GetStaticMeshMaterial) (Variables|Default|GetCustom_OverrideMaterial))
                  (:Out
                    (Rendering|Material|SetMaterial _returnvalue_12 0 _asmaterial_interface)))))))
        (:then_1
          (if (Variables|Default|GetIsOceanExtendOuter)
            (bind _aswater_zone (Utilities|Casting|CastToWaterZone (Components|GetOwner)))
            (bind _asmaterial_interface_2 (Utilities|CustomAsyncLoad (Variables|Default|GetStaticMeshMaterial) (Variables|Default|GetCustom_OverrideMaterial))
              (:Out
                (Rendering|SetFarMeshMaterial _aswater_zone _asmaterial_interface_2)))))
        (:then_2
          (if (Variables|Default|GetIsOceanExtend)
            (Utilities|IsValid _returnvalue_2
              (:Is%20Valid
                (bind _asmaterial_interface_3 (Utilities|CustomAsyncLoad (Variables|Default|GetOceanMaterial_Extend) (Variables|Default|GetCustom_OverrideMaterial))
                  (:Out
                    (Rendering|SetWaterMaterial _returnvalue_2 _asmaterial_interface_3)))))))
        (:then_3
          (if (Variables|Default|GetIsLake)
            (Utilities|IsValid _returnvalue_8
              (:Is%20Valid
                (bind _asmaterial_interface_4 (Utilities|CustomAsyncLoad (Variables|Default|GetLakeMaterial) (Variables|Default|GetCustom_OverrideMaterial))
                  (:Out
                    (Rendering|SetWaterMaterial _returnvalue_8 _asmaterial_interface_4)))))))
        (:then_4
          (if (Variables|Default|GetIsOcean)
            (Utilities|IsValid _returnvalue_13
              (:Is%20Valid
                (bind _asmaterial_interface_5 (Utilities|CustomAsyncLoad (Variables|Default|GetOceanMaterial) (Variables|Default|GetCustom_OverrideMaterial))
                  (:Out
                    (Rendering|SetWaterMaterial _returnvalue_13 _asmaterial_interface_5)))))))
        (:then_5
          (if (Variables|Default|GetIsRiver)
            (Utilities|IsValid _returnvalue_14
              (:Is%20Valid
                (Utilities|FlowControl|Sequence
                  (:then_0
                    (bind _asmaterial_interface_6 (Utilities|CustomAsyncLoad (Variables|Default|GetRiverMaterial) (Variables|Default|GetCustom_OverrideMaterial))
                      (:Out
                        (Rendering|SetWaterMaterial _returnvalue_14 _asmaterial_interface_6))))
                  (:then_1
                    (bind _asmaterial_interface_7 (Utilities|CustomAsyncLoad (Variables|Default|GetLakeTransitionMaterial) (Variables|Default|GetTo_Lake_Custom_OverrideMaterial))
                      (:Out
                        (bind _asmaterial_interface_8 (Utilities|CustomAsyncLoad (Variables|Default|GetOceanTransitionMaterial) (Variables|Default|GetTo_Ocean_Custom_OverrideMaterial))
                          (:Out
                            (Rendering|SetLakeandOceanTransitionMaterials _returnvalue_14 _asmaterial_interface_7 _asmaterial_interface_8)))))))))))
        (:then_6
          (if (Variables|Default|GetIsPool)
            (Utilities|IsValid _returnvalue_11
              (:Is%20Valid
                (bind _asmaterial_interface_9 (Utilities|CustomAsyncLoad (Variables|Default|GetPoolMaterial) (Variables|Default|GetCustom_OverrideMaterial))
                  (:Out
                    (Rendering|SetWaterMaterial _returnvalue_11 _asmaterial_interface_9)))))))
        (:then_7
          (if (Variables|Default|GetIsWaterfall)
            (Utilities|IsValid _returnvalue_4
              (:Is%20Valid
                (Rendering|SetVisibility _returnvalue_4 false false)))))))))
