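; WaterBody_Landmass is an actor Blueprint in FortWater, apparently a water body tied to Landmass terrain, with no graph logic of its own.
; BeginPlay, ActorBeginOverlap, Tick and the Construction Script are empty stubs.
; The custom events Drawing Mode and End Drawing are declared without bodies, so the file does not show what they are meant to do.
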(event EventBeginPlay)

(event Collision|EventActorBeginOverlap
  (params (OtherActor :type Actor%20Object%20Reference)))

(event EventTick
  (params (DeltaSeconds :type Float%20%28single-precision%29)))

(event Custom|Drawing%20Mode)

(event Custom|End%20Drawing)

(fn ConstructionScript)
