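# Material function computing a single deep-water Gerstner wave for water surface displacement and normals in FortWater.
# Wavelength gives wavenumber k, frequency follows sqrt(g·k), and phase is direction·k·position minus frequency times water time.
# Sine and cosine of phase, with Q from steepness/(amplitude·k), build WPO and the normal's XY term (Z zero).
# Position, time, sine, cosine and Q are overridable inputs; phase, time, Q and 1D offset are also output.
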MaterialFunction /FortWater/Materials/Functions/MF_Gerstner_Wave_Deep
graph {nodes [
  n1 = WorldPosition() {worldPositionShaderOffset=WPT_ExcludeAllShaderOffsets}
  n2 = FunctionInput() {bUsePreviewValueAsDefault=true, inputName=2D Direction Vector, inputType=FunctionInput_Vector2, previewValue=<1 0 0 1>}
  n3 = FunctionInput() {bUsePreviewValueAsDefault=true, inputName=Wave Length, inputType=FunctionInput_Scalar, sortPriority=1, previewValue=<100 0 0 1>}
  n4 = Divide(A=MaterialFunctionCall(Multiplier=Constant() {r=2}) {materialFunction=&Pi}.Result, B=@n3) {desc=k}
  n5 = DotProduct(
   A=ComponentMask(None=FunctionInput(Preview=@n1.XYZ) {bUsePreviewValueAsDefault=true, inputName=Pos, sortPriority=8}) {g=true, r=true}
   B=Multiply(A=@n2, B=@n4) {desc=K}
  )
  n6 = ScalarParameter() {defaultValue=980, group=Technical, parameterName=Gravitational Constant}
  n7 = FunctionInput(Preview=MaterialFunctionCall() {materialFunction=/Water/Materials/Functions/GetWaterTime}.WaterTime) {bUsePreviewValueAsDefault=true, inputName=Manual Time, inputType=FunctionInput_Scalar, sortPriority=15, previewValue=<1 0 0 1>}
  n8 = Multiply(A=SquareRoot(None=Multiply(A=@n4, B=@n6)) {desc=w}, B=@n7) {desc=wt}
  MaterialExpressionReroute_4 = Reroute(None=FunctionOutput(None=Subtract(A=@n5, B=@n8)) {outputName=WavePos, sortPriority=10})
  n9 = Custom!(x=@MaterialExpressionReroute_4) {
   code="cos(x);"
   description="cos(x)"
   outputType=CMOT_Float1
   inputs [{inputName=x, input {expression=&self:MaterialExpressionReroute_4}}]
  }
  n10 = FunctionInput(Preview=@n9) {bUsePreviewValueAsDefault=true, inputName=wavecos, inputType=FunctionInput_Scalar, sortPriority=11}
  n11 = MaterialFunctionCall(Z=@n10) {materialFunction=&MakeFloat3}
  n12 = FunctionInput() {bUsePreviewValueAsDefault=true, inputName=Amplitude, inputType=FunctionInput_Scalar, sortPriority=2, previewValue=<10 0 0 1>}
  n13 = FunctionInput() {bUsePreviewValueAsDefault=true, inputName=Wave Steepness, inputType=FunctionInput_Scalar, sortPriority=4, previewValue=<0.8 0 0 1>}
  n14 = Multiply(A=@n12, B=@n4) {desc=wKA}
  n15 = FunctionInput(Preview=Divide(A=@n13, B=@n14) {desc=Q}) {bUsePreviewValueAsDefault=true, inputName=Q, inputType=FunctionInput_Scalar, sortPriority=20}
  n16 = Custom(x=@MaterialExpressionReroute_4) {
   code="sin(x);"
   description="Sin(x)"
   outputType=CMOT_Float1
   inputs [{inputName=x, input {expression=&self:MaterialExpressionReroute_4}}]
  }
  n17 = FunctionInput(Preview=@n16) {bUsePreviewValueAsDefault=true, inputName=wavesin, inputType=FunctionInput_Scalar, sortPriority=10}
  n18 = Multiply(A=Multiply(A=Multiply(A=@n15, B=@n12) {desc=q*amplitude}, B=@n17) {desc=*sin}, B=@n2) {desc=*direction}
  n19 = Constant()
  FunctionOutput(None=Subtract(A=Multiply(A=@n11.Result, B=@n12) {desc="[0,0,Cos]*amplitude"}, B=AppendVector(A=@n18, B=@n19)) {desc=WPO}) {outputName=WPO, sortPriority=1}
  FunctionOutput(None=@n8) {outputName=time, sortPriority=4}
  n20 = Multiply(A=@n14, B=@n17)
  FunctionOutput(None=AppendVector(A=Multiply(A=@n20, B=@n2), B=Constant()) {desc=Normal}) {outputName=Normal}
  n21 = Multiply(A=Multiply(A=FunctionOutput(None=@n15) {outputName=Q, sortPriority=21}) {constB=-1}, B=@n17)
  FunctionOutput(None=@n21) {outputName=1D Offset, sortPriority=20}
  n22 = MaterialFunctionCall(Float2=@n2) {materialFunction=&BreakOutFloat2Components}
  MaterialFunctionCall(X=Multiply(A=@n21, B=@n22.R), Y=Multiply(A=@n21, B=@n22.G), Z=Multiply(A=@n10, B=@n12)) {materialFunction=&MakeFloat3}
 ]}
