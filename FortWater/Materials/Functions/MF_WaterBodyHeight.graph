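# Material function that returns the water Z height at a position by sampling the WaterVelocityAndHeight texture's blue channel.
# It maps the position's XY into clamped UVs using the landscape location and render-target world size from the MPC_Landscape parameter collection.
# Position defaults to world position without shader offsets, and the texture input defaults to a black-filled parameter.
# The sample uses an explicit mip level from a scalar parameter and a shared wrap sampler.
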MaterialFunction /FortWater/Materials/Functions/MF_WaterBodyHeight
graph {nodes [
  n1 = FunctionInput(Preview=WorldPosition() {worldPositionShaderOffset=WPT_ExcludeAllShaderOffsets}.XYZ) {bUsePreviewValueAsDefault=true, inputName=Position, sortPriority=3}
  n2 = CollectionParameter() {collection=/Landmass/Landscape/BlueprintBrushes/MPC/MPC_Landscape, parameterName=LandscapeLocation}
  n3 = Divide(
   A=Subtract(A=ComponentMask(None=@n1) {g=true, r=true}, B=ComponentMask(None=@n2) {g=true, r=true})
   B=ComponentMask(None=CollectionParameter() {collection=&MPC_Landscape, parameterName=RTWorldSize}) {g=true, r=true}
  )
  n4 = TextureObjectParameter() {group=Technical, parameterName=WaterVelocityAndHeight, texture=&Black}
  FunctionOutput(None=TextureSample!(
    UVs=Saturate(None=@n3)
    Tex=FunctionInput(Preview=@n4) {bUsePreviewValueAsDefault=true, inputName=Texture, inputType=FunctionInput_Texture2D, sortPriority=1}
    Level=ScalarParameter() {group=Technical, parameterName=WaterVelocityAndHeightMipLevel}
    Apply View MipBias
   ) {mipValueMode=TMVM_MipLevel, samplerSource=SSM_Wrap_WorldGroupSettings, texture=&Black}.B) {outputName=Water Z Height, sortPriority=4}
 ]}
