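# Material function that adds an extra World Position Offset to an incoming set of material attributes.
# It reads the existing World Position Offset from the Attributes input, adds the Additional WPO input, and writes the sum back.
# All other attributes pass through unchanged, so offsets can be layered without replacing earlier displacement.
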MaterialFunction /FortWater/Materials/Functions/AttributesAddWPO
graph {nodes [
  n1 = GetMaterialAttributes(None=FunctionInput() {inputName=Attributes, inputType=FunctionInput_MaterialAttributes}) {attributeGetTypes [#1]}
  n2 = FunctionInput() {bUsePreviewValueAsDefault=true, inputName=Additional WPO, sortPriority=1, previewValue=<1 0 0 1>}
  FunctionOutput(None=SetMaterialAttributes!(MaterialAttributes=@n1.MaterialAttributes, WorldPositionOffset=Add(A=@n1.WorldPositionOffset, B=@n2)) {attributeSetTypes [#1]})
 ]}
