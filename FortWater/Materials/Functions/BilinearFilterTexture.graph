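# Material function in the FortWater material library that manually bilinear-filters a Texture2DArray, for filtering at higher precision than the stored format.
# It converts UVs to texel space using the Resolution input, snaps to texel centers, and derives one-texel offsets.
# Four clamped samples at the given slice and explicit mip level are blended by nested lerps using the fractional texel position, outputting RGBA.
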MaterialFunction /FortWater/Materials/Functions/BilinearFilterTexture
graph {
 properties {
  bExposeToLibrary=true
  description="Manual Bilinear Filtering is useful when you require filtering at higher precision than the stored format. "
 }
 nodes [
  n1 = TextureCoordinate()
  n2 = FunctionInput() {bUsePreviewValueAsDefault=true, inputName=Resolution, inputType=FunctionInput_Vector2, sortPriority=2, previewValue=<512 512 0 1>}
  n3 = Subtract(A=Multiply(A=FunctionInput(Preview=@n1) {bUsePreviewValueAsDefault=true, inputName=UVs, inputType=FunctionInput_Vector2, sortPriority=1}, B=@n2)) {constB=0.5}
  n4 = Divide(A=Add(A=Floor(None=@n3)) {constB=0.5}, B=@n2)
  n5 = FunctionInput() {bUsePreviewValueAsDefault=true, inputName=Slice Index, inputType=FunctionInput_Scalar, sortPriority=4}
  n6 = FunctionInput() {bUsePreviewValueAsDefault=true, inputName=Mip Level, inputType=FunctionInput_Scalar, sortPriority=3}
  n7 = FunctionInput(Preview=TextureObject() {samplerType=SAMPLERTYPE_LinearColor, texture=/Water/Textures/BlackTexture2DArray}) {inputName=Texture, inputType=FunctionInput_Texture2DArray}
  n8 = TextureSample!(UVs=AppendVector(A=@n4, B=@n5), Tex=@n7, Level=@n6, Apply View MipBias) {desc="0,0", mipValueMode=TMVM_MipLevel, samplerSource=SSM_Clamp_WorldGroupSettings, samplerType=SAMPLERTYPE_LinearColor}
  n9 = Divide(B=@n2)
  n10 = MaterialFunctionCall(Float2=@n9) {materialFunction=&BreakOutFloat2Components}
  n11 = Constant()
  n12 = AppendVector(A=@n10.R, B=@n11)
  n13 = TextureSample(UVs=AppendVector(A=Add(A=@n4, B=@n12), B=@n5), Tex=@n7, Level=@n6) {desc="1,0", mipValueMode=TMVM_MipLevel, samplerSource=SSM_Clamp_WorldGroupSettings, samplerType=SAMPLERTYPE_LinearColor}
  n14 = MaterialFunctionCall(Float2=Frac(None=@n3)) {materialFunction=&BreakOutFloat2Components}
  n15 = AppendVector(A=@n11, B=@n10.G)
  n16 = TextureSample(UVs=AppendVector(A=Add(A=@n4, B=@n15), B=@n5), Tex=@n7, Level=@n6) {desc="0,1", mipValueMode=TMVM_MipLevel, samplerSource=SSM_Clamp_WorldGroupSettings, samplerType=SAMPLERTYPE_LinearColor}
  n17 = TextureSample(UVs=AppendVector(A=Add(A=@n4, B=@n9), B=@n5), Tex=@n7, Level=@n6) {desc="1,1", mipValueMode=TMVM_MipLevel, samplerSource=SSM_Clamp_WorldGroupSettings, samplerType=SAMPLERTYPE_LinearColor}
  FunctionOutput(None=LinearInterpolate(A=LinearInterpolate(A=@n8.RGBA, B=@n13.RGBA, Alpha=@n14.R), B=LinearInterpolate(A=@n16.RGBA, B=@n17.RGBA, Alpha=@n14.R), Alpha=@n14.G))
 ]
}
