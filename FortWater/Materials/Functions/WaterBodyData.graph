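# Material function that supplies water height, terrain height, depth and flow velocity with speed to FortWater materials.
# A scalar or tile-instance body index resolves the water zone, whose texture array slice is sampled at the zone-normalized position.
# High quality uses manual bilinear filtering; Low and Medium use plain samples.
# Custom HLSL decodes the sample, with static switches selecting fixed body data, zone ground minimum, vertex-color flow or parameter overrides instead.
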MaterialFunction /FortWater/Materials/Functions/WaterBodyData
graph {nodes [
  n1 = TextureObjectParameter() {group=Technical, parameterName=WaterVelocityAndHeight, samplerType=SAMPLERTYPE_LinearColor, texture=/Water/Textures/BlackTexture2DArray}
  n2 = FunctionInput(Preview=@n1) {bUsePreviewValueAsDefault=true, inputName=Texture, inputType=FunctionInput_Texture2DArray, sortPriority=1}
  n3 = FunctionInput(Preview=WorldPosition() {worldPositionShaderOffset=WPT_ExcludeAllShaderOffsets}.XYZ) {bUsePreviewValueAsDefault=true, inputName=Position, inputType=FunctionInput_Vector2, sortPriority=3}
  n4 = MaterialFunctionCall() {materialFunction=/Water/Materials/Functions/WaterZoneData}
  n5 = Subtract(A=@n4.Location, B=Divide(A=@n4.Extent))
  n6 = Saturate(None=Divide(A=Subtract(A=@n3, B=@n5), B=@n4.Extent))
  MaterialExpressionStaticSwitchParameter_8 = StaticSwitchParameter(
   True=ScalarParameter() {parameterName=WaterBodyIndex}
   False=Custom() {code="GetWaterWaveParamIndex(Parameters)", description=WaterBodyIndex From Tile Instance, outputType=CMOT_Float1, inputs []}
  ) {parameterName=UseScalarWaterBodyIndex}
  MaterialExpressionCustom_6 = Custom!(WaterBodyIndex=@MaterialExpressionStaticSwitchParameter_8) {
   code="return GetWaterBodyData((int)WaterBodyIndex).WaterZoneIndex;"
   description=GetWaterZoneIndexForWaterBody
   outputType=CMOT_Float1
   includeFilePaths [/Plugin/Water/Private/WaterDataFunctions.ush]
   inputs [{inputName=WaterBodyIndex, input {expression=&self:MaterialExpressionStaticSwitchParameter_8}}]
  }
  MaterialExpressionReroute_2 = Reroute(None=@MaterialExpressionCustom_6)
  n7 = Custom!(WaterZoneIndex=@MaterialExpressionReroute_2) {
   code="return GetWaterInfoTextureViewIndex(WaterZoneIndex);"
   description=GetWaterInfoTextureViewIndex
   outputType=CMOT_Float1
   showCode=true
   includeFilePaths [&ush]
   inputs [{inputName=WaterZoneIndex, input {expression=&self:MaterialExpressionReroute_2}}]
  }
  n8 = MaterialFunctionCall!(Texture=@n2, UVs=@n6, Resolution=TextureProperty(Tex=@n2), Mip Level, Slice Index=@n7) {materialFunction=&/BilinearFilterTexture}
  n9 = TextureSample!(
   UVs=AppendVector(A=@n6, B=@n7)
   Tex=@n2
   Level=ScalarParameter() {group=Technical, parameterName=WaterVelocityAndHeightMipLevel}
   Apply View MipBias
  ) {mipValueMode=TMVM_MipLevel, samplerSource=SSM_Clamp_WorldGroupSettings, texture=&Black}
  MaterialExpressionReroute_1 = Reroute(None=QualitySwitch(Default=@n8.Result, Low=@n9.RGBA, High=@n8.Result, Medium=@n9.RGBA))
  n10 = Custom!(WaterInfoSample=@MaterialExpressionReroute_1, WaterZoneIndex=@MaterialExpressionCustom_6) {
   code="return DecodeWaterInfoZHeight(WaterInfoSample, WaterZoneIndex);"
   description=DecodeWaterInfoZHeight
   outputType=CMOT_Float1
   includeFilePaths [&ush]
   inputs [
    {inputName=WaterInfoSample, input {expression=&self:MaterialExpressionReroute_1}}
    {inputName=WaterZoneIndex, input {expression=&self:MaterialExpressionCustom_6}}
   ]
  }
  n11 = Custom!(WaterBodyIndex=@MaterialExpressionStaticSwitchParameter_8) {
   code="FWaterBodyData WaterBodyData = GetWaterBodyData((int)WaterBodyIndex);\r\n\r\nFixedVelocity = WaterBodyData.FixedVelocity;\r\nFixedZHeight = WaterBodyData.FixedZHeight;\r\nFixedWaterDepth = WaterBodyData.FixedWaterDepth;\r\n\r\nreturn 0.0f;"
   description=GetWaterBodyFallbacks
   additionalOutputs [
    {outputName=FixedVelocity, outputType=CMOT_Float3}
    {outputName=FixedZHeight, outputType=CMOT_Float1}
    {outputName=FixedWaterDepth, outputType=CMOT_Float1}
   ]
   includeFilePaths [&ush]
   inputs [{inputName=WaterBodyIndex, input {expression=&self:MaterialExpressionStaticSwitchParameter_8}}]
  }
  n12 = StaticSwitchParameter(True=@n10, False=@n11.FixedZHeight) {defaultValue=true, parameterName=UseWaterInfoTexture}
  n13 = Custom!(WaterInfoSample=@MaterialExpressionReroute_1, WaterZoneIndex=@MaterialExpressionCustom_6) {
   code="return DecodeWaterInfoGroundHeight(WaterInfoSample, WaterZoneIndex);"
   description=DecodeWaterInfoGroundHeight
   outputType=CMOT_Float1
   includeFilePaths [&ush]
   inputs [
    {inputName=WaterInfoSample, input {expression=&self:MaterialExpressionReroute_1}}
    {inputName=WaterZoneIndex, input {expression=&self:MaterialExpressionCustom_6}}
   ]
  }
  n14 = MaterialFunctionCall() {materialFunction=&WaterZoneData}
  n15 = Subtract(
   A=FunctionOutput(None=StaticSwitchParameter(True=@n11.FixedZHeight, False=@n12) {parameterName=UseFixedZ}) {outputName=Water Z Height, sortPriority=4}
   B=FunctionOutput(None=StaticSwitchParameter(True=@n13, False=@n14.GroundZMin) {defaultValue=true, parameterName=UseWaterInfoTexture}) {outputName=Terrain Z Height, sortPriority=99}
  )
  FunctionOutput(None=StaticSwitchParameter(
    True=ScalarParameter() {defaultValue=1000, parameterName=Overridden Water Depth}
    False=StaticSwitchParameter(True=@n15, False=@n11.FixedWaterDepth) {defaultValue=true, parameterName=UseWaterInfoTexture}
   ) {parameterName=Override Water Depth}) {outputName=Water Depth, sortPriority=3}
  n16 = VectorParameter() {parameterName=Overridden Velocity}
  MaterialExpressionScalarParameter_9 = ScalarParameter() {defaultValue=1024, parameterName=MaxFlowVelocity}
  n17 = Custom!(WaterInfoSample=@MaterialExpressionReroute_1, MaxVelocity=@MaterialExpressionScalarParameter_9) {
   code="return DecodeWaterInfoVelocity(WaterInfoSample, MaxVelocity);"
   description=Decode Flow From Water Info
   outputType=CMOT_Float2
   includeFilePaths [&ush]
   inputs [
    {inputName=WaterInfoSample, input {expression=&self:MaterialExpressionReroute_1}}
    {inputName=MaxVelocity, input {expression=&self:MaterialExpressionScalarParameter_9}}
   ]
  }
  MaterialExpressionVertexColor_1 = VertexColor()
  n18 = Custom!(Flow=@MaterialExpressionVertexColor_1, MaxVelocity=@MaterialExpressionScalarParameter_9) {
   code="float Magnitude = Flow.x * MaxVelocity;\r\nfloat Direction = Flow.y * 2 * PI;\r\n\r\nfloat FlowX = Magnitude * cos(Direction);\r\nfloat FlowY = Magnitude * sin(Direction);\r\n\r\nreturn float2(FlowX, FlowY);"
   description=Denormalize Flow From Vertex Color
   outputType=CMOT_Float2
   inputs [
    {
     inputName=Flow
     input {expression=&self:MaterialExpressionVertexColor_1, mask=1, maskB=1, maskG=1, maskR=1}
    }
    {inputName=MaxVelocity, input {expression=&self:MaterialExpressionScalarParameter_9}}
   ]
  }
  n19 = StaticSwitchParameter(
   True=MaterialFunctionCall(X=@n16.R, Y=@n16.R) {materialFunction=&MakeFloat2}.Result
   False=StaticSwitchParameter(True=@n17, False=@n18) {defaultValue=true, parameterName=UseWaterInfoTexture}
  ) {parameterName=Override Velocity}
  FunctionOutput(None=AppendVector(A=@n19, B=MaterialFunctionCall(Vector 2=@n19) {materialFunction=/Engine/Functions/Engine_MaterialFunctions02/Utility/VectorLength}.V2 Length)) {outputName=VelocityAndLength}
 ]}
