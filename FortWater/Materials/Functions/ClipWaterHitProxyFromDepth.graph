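# Material function that clips a water material's editor hit proxy to the visible outline of the water, so selection matches its shape.
# It reads Water Depth from WaterBodyData, saturates it and remaps it into a signed clip value starting at -1.
# A Custom HLSL node clips on that value only in the hit proxy shader.
# It passes BaseColor through unchanged, which wires the clip into the material attributes.
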MaterialFunction /FortWater/Materials/Functions/ClipWaterHitProxyFromDepth
graph {nodes [
  MaterialExpressionGetMaterialAttributes_0 = GetMaterialAttributes(None=FunctionInput() {inputName=Attributes, inputType=FunctionInput_MaterialAttributes}) {attributeGetTypes [#1]}
  MaterialExpressionLinearInterpolate_3 = LinearInterpolate(Alpha=Saturate(None=MaterialFunctionCall!(Texture, Position) {materialFunction=&/WaterBodyData}.Water Depth)) {constA=-1}
  n1 = Custom!(PassThrough=@MaterialExpressionGetMaterialAttributes_0.BaseColor, x=@MaterialExpressionLinearInterpolate_3) {
   code="##if HIT_PROXY_SHADER\r\nclip(x);\r\n##endif\r\nreturn PassThrough;"
   description=Hit Proxy Depth-Clip
   inputs [
    {
     inputName=PassThrough
     input {expression=&self:MaterialExpressionGetMaterialAttributes_0, outputIndex=1}
    }
    {inputName=x, input {expression=&self:MaterialExpressionLinearInterpolate_3}}
   ]
  }
  FunctionOutput(None=SetMaterialAttributes!(MaterialAttributes=@MaterialExpressionGetMaterialAttributes_0.MaterialAttributes, BaseColor=@n1) {attributeSetTypes [#1]}) {description=Clips the hit proxy to match the outline of the water with pixel-perfect accuracy}
 ]}
