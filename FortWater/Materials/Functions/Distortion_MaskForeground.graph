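# Water material function that corrects distorted refraction screen UVs so foreground objects do not bleed into the distortion.
# It keeps the input UV only where scene depth sampled there lies behind the pixel, otherwise using undistorted viewport UV.
# The result then blends toward undistorted UV as offset scene depth exceeds distorted-sample depth, scaled by a kill depth.
# Both offset and kill depth are Refraction parameters.
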MaterialFunction /FortWater/Materials/Functions/Distortion_MaskForeground
graph {nodes [
  n1 = FunctionInput() {bUsePreviewValueAsDefault=true, inputName=Screen Position, inputType=FunctionInput_Vector2}
  n2 = SceneDepth(UVs=@n1)
  n3 = Subtract(A=@n2.R, B=PixelDepth().R)
  n4 = ScreenPosition()
  n5 = SceneDepth()
  n6 = ScalarParameter() {defaultValue=100, group=Refraction, parameterName=Refraction Scene Depth Offset}
  n7 = Divide(A=Subtract(A=Add(A=@n5.R, B=@n6), B=@n2.R), B=ScalarParameter() {defaultValue=256, group=Refraction, parameterName=Refraction Kill Depth})
  FunctionOutput(None=LinearInterpolate(A=If(A=@n3, "A > B"=@n1, "A == B"=@n4.ViewportUV, "A < B"=@n4.ViewportUV), B=@n4.ViewportUV, Alpha=Saturate(None=@n7)))
 ]}
