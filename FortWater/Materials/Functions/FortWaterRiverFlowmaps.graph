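# FortWaterRiverFlowmaps is a material function that adds flowing river waves and foam to incoming water material attributes.
# Velocity and depth drive a WaterFlowMap call that yields animated normals and height, with exponential depth masks from one vectorized Custom node.
# Velocity, depth, world-space noise and sim erasure shape foam masks and scattering.
# Quality and shading-path switches simplify foam and phase timing on cheaper paths; DisableRiverFoam keeps only normals.
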MaterialFunction /FortWater/Materials/Functions/FortWaterRiverFlowmaps
graph {nodes [
  n1 = ComponentMask(None=NamedRerouteUsage()) {b=true}
  n2 = Subtract(
   A=Multiply(A=Divide(A=@n1) {constB=100}, B=ScalarParameter() {defaultValue=0.4, group=Flowmaps, parameterName=Velocity Foam Amount})
   B=Multiply(A=NamedRerouteUsage(), B=ScalarParameter() {defaultValue=0.002, parameterName=Depth Foam Removal})
  )
  n3 = Divide(
   A=ComponentMask(None=WorldPosition() {worldPositionShaderOffset=WPT_ExcludeAllShaderOffsets}.XYZ) {g=true, r=true}
   B=ScalarParameter() {defaultValue=3240, parameterName=Flowmap Breakup Scale}
  )
  n4 = TextureSample!(UVs=@n3, Tex, Level=ScalarParameter() {defaultValue=2.5, parameterName=River Foam Phase Mip Level}, Apply View MipBias) {mipValueMode=TMVM_MipLevel, samplerSource=SSM_Wrap_WorldGroupSettings, texture=/FortWater/Textures/T_WaterFlow_01_Foam_Tiled}
  n5 = ScalarParameter() {parameterName=RiverFoamWorldSpaceRandomEdges}
  n6 = ScalarParameter() {parameterName=RiverFoamWorldSpaceRandomCenter}
  n7 = Saturate(None=Divide(A=NamedRerouteUsage(), B=ScalarParameter() {defaultValue=200, parameterName=Random Foam DepthRange}))
  Add(A=@n2, B=Multiply(A=@n4.G, B=LinearInterpolate(A=@n5, B=@n6, Alpha=@n7)))
  Constant3Vector()
  n8 = NamedRerouteUsage()
  n9 = ScalarParameter() {group=Flowmaps, parameterName=Wave Z Displacement}
  FunctionOutput(None=AppendVector(A=Constant2Vector(), B=Multiply(A=@n8, B=@n9))) {outputName=Flowmap WPO, sortPriority=1}
  n10 = GetMaterialAttributes(None=FunctionInput() {inputName=Attributes, inputType=FunctionInput_MaterialAttributes}) {attributeGetTypes [
    #1
    #2
    #3
    #4
    #5
    #6
   ]}
  n11 = Constant3Vector() {constant=<rgba 1 1 0 0>}
  n12 = Multiply(A=NamedRerouteUsage(), B=@n11)
  n13 = Add(A=@n10.Normal, B=@n12)
  n14 = LinearInterpolate(A=@n10.BaseColor, B=NamedRerouteUsage(), Alpha=NamedRerouteUsage())
  n15 = NamedRerouteUsage()
  n16 = SetMaterialAttributes!(
   MaterialAttributes=@n10.MaterialAttributes
   BaseColor=@n14
   Specular=LinearInterpolate(A=@n10.Specular, B=ScalarParameter() {defaultValue=0.5, group=Foam and Pondscum, parameterName=River Foam Specular}, Alpha=@n15)
   Roughness=LinearInterpolate(A=@n10.Roughness, B=ScalarParameter() {defaultValue=0.5, group=Foam and Pondscum, parameterName=River Foam Roughness}, Alpha=@n15)
   Opacity=LinearInterpolate(A=@n10.Opacity, B=ScalarParameter() {defaultValue=1, group=Foam and Pondscum, parameterName=River Foam Opacity}, Alpha=@n15)
   Normal=@n13
   WorldPositionOffset=Add(A=@n10.WorldPositionOffset, B=NamedRerouteUsage())
  ) {attributeSetTypes [
    #1
    #2
    #3
    #4
    #5
    #6
   ]}
  n17 = SetMaterialAttributes(
   MaterialAttributes=@n10.MaterialAttributes
   BaseColor=@n14
   Specular=@n10.MaterialAttributes
   Roughness=@n10.MaterialAttributes
   Opacity=@n10.MaterialAttributes
   Normal=@n13
   WorldPositionOffset=@n10.MaterialAttributes
  ) {attributeSetTypes [
    #1
    #2
    #3
    #4
    #5
    #6
   ]}
  n18 = QualitySwitch(Default=@n16, Low=@n17, High=@n16, Medium=@n17)
  FunctionOutput(None=StaticSwitchParameter(
    True=SetMaterialAttributes!(MaterialAttributes=@n10.MaterialAttributes, Normal=@n13) {attributeSetTypes [#5]}
    False=ShadingPathSwitch(Default=@n18, Deferred=@n18, Forward=@n17, Mobile=@n17)
   ) {group=Foam and Pondscum, parameterName=DisableRiverFoam}) {outputName=Attributes}
  n19 = FunctionInput() {bUsePreviewValueAsDefault=true, inputName=Sim Foam Erasure, inputType=FunctionInput_Scalar, sortPriority=1}
  n20 = ScalarParameter() {defaultValue=-1.838, group=Flowmaps, parameterName=Foam_SimErasure_Min}
  n21 = Subtract(A=@n2, B=Multiply(A=Add(A=@n19, B=@n20), B=ScalarParameter() {defaultValue=0.152, group=Flowmaps, parameterName=Foam_SimErasure_Range}))
  n22 = ScalarParameter() {defaultValue=4.5, parameterName=River Foam Phase Mip Scatter}
  n23 = Subtract(
   A=Add(
    A=@n21
    B=TextureSample(UVs=@n3, Level=@n22) {mipValueMode=TMVM_MipLevel, samplerSource=SSM_Wrap_WorldGroupSettings, texture=&T_WaterFlow_01_Foam_Tiled}.R
   )
   B=ScalarParameter() {defaultValue=1.25, parameterName=Foam Scattering Offset}
  )
  n24 = VectorParameter() {parameterName=River Foam Scattering, defaultValue=<rgba 0.6 1 1 0.4>}
  n25 = Multiply(A=@n24.RGB, B=Divide(A=@n24.A) {constB=100})
  FunctionOutput(None=Saturate(None=Multiply(A=@n23, B=@n25))) {outputName=Scattering, sortPriority=5}
  n26 = VectorParameter() {parameterName=River Foam Color 1, defaultValue=<rgba 0.7396 0.9349 1 0.45>}
  FunctionOutput(None=NamedRerouteDeclaration(None=Multiply(A=@n26.RGB, B=@n26.RGB)) {name=FinalFoam Color, nodeColor=<rgba 1 0.4922 0 1>}) {outputName=Foam, sortPriority=10}
  n27 = NamedRerouteUsage()
  FunctionOutput(None=AppendVector(A=NamedRerouteUsage(), B=ComponentMask(None=@n27) {r=true})) {
   description="Outputs 4 exponential masks using a single vectorized Exp operation. R=Wave Masked based on Depth, G=Flowmap Mask based on Depth, B=Coloration Mask based on Depth, A=Flow Mask based on velocity"
   outputName=Phase Offset And Foam
   sortPriority=10
  }
  Multiply(A=@n10.Normal, B=@n11) {desc="this will cause a seam on Lake to river transition\r\nmultiply the Flow Map Normals instead"}
  n28 = Multiply(A=NamedRerouteUsage(), B=NamedRerouteUsage())
  n29 = Add(
   A=Multiply(
    A=NamedRerouteDeclaration(None=@n28) {name=Height Scaled by Velocity, nodeColor=<rgba 0 1 0.2734 1>}
    B=ScalarParameter() {defaultValue=0.2, parameterName=WPO River Foam}
   )
   B=Multiply(A=NamedRerouteUsage(), B=ScalarParameter() {defaultValue=0.5, parameterName=Foam Texture Amount})
  )
  n30 = Multiply(A=NamedRerouteUsage(), B=ScalarParameter() {defaultValue=1, parameterName=Phase offset Foam Amount})
  n31 = Subtract(A=Add(A=Add(A=@n29, B=@n30), B=@n21), B=ScalarParameter() {defaultValue=1.7, group=Flowmaps, parameterName=River Foam offset})
  n32 = Multiply(A=@n31, B=ScalarParameter() {defaultValue=1, parameterName=River Foam Boost})
  NamedRerouteDeclaration(None=FunctionOutput(None=Saturate(None=@n32)) {outputName=Foam Mask, sortPriority=10}) {name=FoamMask, nodeColor=<rgba 0.125 0 1 1>}
  n33 = Max(A=@n31) {constB=0}
  n34 = ScalarParameter() {defaultValue=16, group=Foam and Pondscum, parameterName=Foam WPO}
  NamedRerouteDeclaration(None=Multiply(A=Multiply(A=@n33, B=@n34), B=Constant3Vector() {constant=<rgba 0 0 1 0>})) {name=Foam WPO, nodeColor=<rgba 0.7812 1 0 1>}
  n35 = TextureObjectParameter() {group=Flowmaps, parameterName=FlowmapWaveNormalHeight, texture=/Game/Global/Textures/Default/Tileable_Patterns/T_RiverFoam_new_M}
  n36 = NamedRerouteUsage()
  n37 = ScalarParameter() {defaultValue=1250, group=Flowmaps, parameterName=FlowmapWaveNormalScale}
  n38 = Multiply(A=ScalarParameter() {defaultValue=1.5, parameterName=Final Flow Mult}, B=Divide(A=ComponentMask(None=@n36) {g=true, r=true}, B=@n37))
  n39 = ScalarParameter() {defaultValue=0.75, group=Flowmaps, parameterName=Phase Fade Time}
  n40 = MaterialFunctionCall(Time=MaterialFunctionCall() {materialFunction=/Water/Materials/Functions/GetWaterTime}.WaterTime, Speed=@n39) {materialFunction=&TimeWithSpeedVariable}
  n41 = Add(
   A=@n40.Time
   B=Multiply(A=NamedRerouteUsage(), B=ScalarParameter() {defaultValue=-0.15, group=Flowmaps, parameterName=Water Flowmap offset Amount})
  ) {constB=0}
  n42 = QualitySwitch(Default=@n41, Low=@n40.Time, High=@n41, Medium=@n40.Time)
  n43 = MaterialFunctionCall!(
   Diffuse=@n35
   Normal Map
   "Flow Vector Map (see tooltip)"=LinearInterpolate(A=Divide(A=@n38, B=@n39), Alpha=NamedRerouteUsage()) {constB=0}
   UV texture 2 offset and Scale=VectorParameter() {group=Flowmaps, parameterName=Flowmap UV2 Offset and Scale, defaultValue=<rgba 0.5 0.5 1 0>}.RGB
   "Time (see tooltip)"=ShadingPathSwitch(Default=@n42, Deferred=@n42, Forward=@n40.Time, Mobile=@n40.Time)
   UVs=NamedRerouteUsage()
  ) {materialFunction=&/WaterFlowMap}
  NamedRerouteDeclaration(None=@n43.Diffuse Alpha) {name=FlowMap Alpha Channel, nodeColor=<rgba 0.8516 0 1 1>}
  n44 = ConstantBiasScale(None=@n43.Diffuse) {bias=-0.5, scale=2}
  n45 = ComponentMask(None=@n44) {b=true}
  NamedRerouteDeclaration(None=@n45) {name=Flow Map Height Channel, nodeColor=<rgba 1 0 0.8672 1>}
  n46 = ComponentMask(None=@n44) {g=true, r=true}
  NamedRerouteDeclaration(None=@n46) {name=Flow Map Normal RG Channel, nodeColor=<rgba 0.9688 1 0 1>}
  NamedRerouteDeclaration(None=@n4.R) {name=WorldSpaceNoise Texture, nodeColor=<rgba 0.1953 1 0 1>}
  n47 = FunctionInput() {bUsePreviewValueAsDefault=true, inputName=Velocity, sortPriority=2, previewValue=<1 0 0 1>}
  n48 = Multiply(A=NamedRerouteDeclaration(None=@n47) {name=InputVelocity, nodeColor=<rgba 0.8281 1 0 1>}, B=NamedRerouteUsage())
  NamedRerouteDeclaration(None=@n48) {name=VelocityScaled, nodeColor=<rgba 0 1 0.4609 1>}
  n49 = NamedRerouteUsage()
  n50 = Divide(A=Max(A=ComponentMask(None=@n49) {b=true}) {constB=0.001}, B=ScalarParameter() {defaultValue=512, parameterName=Velocity for Max Normals})
  n51 = Multiply(
   A=Multiply(A=NamedRerouteUsage(), B=NamedRerouteUsage())
   B=Multiply(A=ScalarParameter() {defaultValue=0.3, group=Flowmaps, parameterName=WavelNormalIntensity}, B=Saturate(None=@n50))
  )
  NamedRerouteDeclaration(None=FunctionOutput(None=DeriveNormalZ(InXY=@n51)) {outputName=Flowmap Normals}) {name=FlowMapNormals, nodeColor=<rgba 0.9453 0 1 1>}
  n52 = FunctionInput(Preview=Constant()) {bUsePreviewValueAsDefault=true, inputName=Far Mask Override, inputType=FunctionInput_Scalar, sortPriority=3}
  NamedRerouteDeclaration(None=@n52) {name=InputFarMaskOverride, nodeColor=<rgba 0 0.5312 1 1>}
  n53 = Divide(
   A=ComponentMask(None=WorldPosition() {worldPositionShaderOffset=WPT_ExcludeAllShaderOffsets}.XYZ) {g=true, r=true}
   B=LinearInterpolate(A=@n37, B=ScalarParameter() {defaultValue=8000, group=Distance Blending, parameterName=FarScale}, Alpha=NamedRerouteUsage())
  )
  NamedRerouteDeclaration(None=@n53) {name=Flow Map UVs, nodeColor=<rgba 0 1 0.0859 1>}
  n54 = FunctionInput() {bUsePreviewValueAsDefault=true, inputName=Depth, inputType=FunctionInput_Scalar, sortPriority=4, previewValue=<1 0 0 1>}
  n55 = Max(A=@n54) {constB=0}
  NamedRerouteDeclaration(None=@n55) {name=Input Depth, nodeColor=<rgba 0.4531 0 1 1>}
  n56 = ScalarParameter() {defaultValue=256, group=Flowmaps, parameterName=Target Depth for Max Velocity Flow}
  MaterialExpressionDivide_5 = Divide(
   A=NamedRerouteUsage()
   B=AppendVector(A=Divide(A=@n56), B=ScalarParameter() {defaultValue=128, group=Flowmaps, parameterName=Target Depth for Flowmap Waves})
  )
  n57 = Custom!(x=@MaterialExpressionDivide_5) {
   code="return 1-exp(-x);"
   description=exp
   outputType=CMOT_Float2
   inputs [{inputName=x, input {expression=&self:MaterialExpressionDivide_5}}]
  }
  n58 = ComponentMask(None=@n57) {r=true}
  NamedRerouteDeclaration(None=@n58) {name=TargetDepthForMaxVelocity, nodeColor=<rgba 1 0 0.7734 1>}
  n59 = ComponentMask(None=@n57) {g=true}
  NamedRerouteDeclaration(None=@n59) {name=TargetDepth For Flowmap Waves, nodeColor=<rgba 0.2891 1 0 1>}
  NamedRerouteUsage()
  NamedRerouteUsage()
  Reroute(None=@n47)
  Reroute(None=@n53)
  Reroute(None=@n55)
  Reroute(None=@n58)
  Reroute(None=@n59)
  Reroute(None=@n43.Diffuse Alpha)
  Reroute(None=@n45)
  Reroute(None=@n46)
  Reroute(None=@n28)
  Reroute(None=@n4.R)
  Reroute(None=@n48)
  Reroute(None=@n52)
  StaticSwitchParameter(
   True=LinearInterpolate(A=@n10.Normal, B=@n12, Alpha=FunctionInput() {bUsePreviewValueAsDefault=true, inputName=Transition, inputType=FunctionInput_Scalar})
   False=@n13
  ) {
   desc="Disconnected because this breaks fluidsim\r\nThe normal input from fluidsim is applied globally to water, and shouldn't be faded out in any case\r\nRipples will interact with river velocity better in the future\r\nFORT-726284"
   parameterName=River Transition
  }
 ]}
