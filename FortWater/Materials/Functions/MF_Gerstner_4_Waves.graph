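# Material function for water that sums four Gerstner waves, chosen by index offset and water body, into WPO and normal outputs.
# A static switch selects wave data from GetWaveData or MPC_Water parameters for each GerstnerWave_Deep_WithZTrace call.
# The waves' phases are packed into float4s for batched custom sin/cos, then fed back into each wave.
# Normals optionally use vertex interpolation and become flat on low quality, forward and mobile.
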MaterialFunction /FortWater/Materials/Functions/MF_Gerstner_4_Waves
graph {nodes [
  n1 = FunctionInput(Preview=StaticBool()) {bUsePreviewValueAsDefault=true, inputName=Z Offset Only, inputType=FunctionInput_StaticBool}
  n2 = FunctionInput() {bUsePreviewValueAsDefault=true, inputName=Index Offset, inputType=FunctionInput_Scalar, sortPriority=7}
  n3 = FunctionInput(Preview=ScalarParameter() {group=Waves, parameterName=WaterBodyIndex}) {bUsePreviewValueAsDefault=true, inputName=WaterBodyIndex, inputType=FunctionInput_Scalar, sortPriority=5}
  n4 = MaterialFunctionCall!(Wave Index=Add(A=Constant(), B=@n2), Water Body Index=@n3) {materialFunction=&/GetWaveData}
  n5 = CollectionParameter() {collection=/Water/Materials/MPC/MPC_Water, parameterName=k0}
  n6 = StaticBoolParameter() {defaultValue=true, parameterName=UseTextureWaveData}
  n7 = MaterialFunctionCall(Float4=StaticSwitch(True=@n4.Params, False=CollectionParameter() {collection=&MPC_Water, parameterName=w0}, Value=@n6)) {materialFunction=&BreakOutFloat4Components}
  n8 = FunctionInput(Preview=WorldPosition() {worldPositionShaderOffset=WPT_ExcludeAllShaderOffsets}.XYZ) {bUsePreviewValueAsDefault=true, inputName=Position}
  n9 = MaterialFunctionCall(Wave Index=Add(A=Constant() {r=1}, B=@n2), Water Body Index=@n3) {materialFunction=&GetWaveData}
  n10 = CollectionParameter() {collection=&MPC_Water, parameterName=k1}
  n11 = MaterialFunctionCall(Float4=StaticSwitch(True=@n9.Params, False=CollectionParameter() {collection=&MPC_Water, parameterName=w1}, Value=@n6)) {materialFunction=&BreakOutFloat4Components}
  n12 = MaterialFunctionCall(Wave Index=Add(A=Constant() {r=2}, B=@n2), Water Body Index=@n3) {materialFunction=&GetWaveData}
  n13 = CollectionParameter() {collection=&MPC_Water, parameterName=k2}
  n14 = MaterialFunctionCall(Float4=StaticSwitch(True=@n12.Params, False=CollectionParameter() {collection=&MPC_Water, parameterName=w2}, Value=@n6)) {materialFunction=&BreakOutFloat4Components}
  n15 = Custom!(x=@MaterialExpressionMaterialFunctionCall_0.Result) {
   code="cos(x);"
   description="cos(x)"
   outputType=CMOT_Float4
   inputs [{inputName=x, input {expression=&self:MaterialExpressionMaterialFunctionCall_0}}]
  }
  n16 = MaterialFunctionCall(Float4=@n15) {materialFunction=&BreakOutFloat4Components}
  n17 = Custom(x=@MaterialExpressionMaterialFunctionCall_31.Result) {
   code="cos(x);"
   description="cos(x)"
   outputType=CMOT_Float4
   inputs [{inputName=x, input {expression=&self:MaterialExpressionMaterialFunctionCall_31}}]
  }
  n18 = MaterialFunctionCall(Float4=@n17) {materialFunction=&BreakOutFloat4Components}
  n19 = MaterialFunctionCall!(
   Z Only Offset=@n1
   2D Direction Vector=StaticSwitch(True=@n12.Direction, False=ComponentMask(None=@n13) {g=true, r=true}, Value=@n6)
   Wave Length=@n14.R
   Amplitude=@n14.R
   Wave Steepness=@n14.R
   Pos=@n8
   wavesin 1=@n28.B
   wavesin 2=@n25.B
   wavecos 1=@n16.B
   wavecos 2=@n18.B
   Manual Time
  ) {desc=2, materialFunction=&/GerstnerWave_Deep_WithZTrace}
  n20 = MaterialFunctionCall(Wave Index=Add(A=Constant() {r=3}, B=@n2), Water Body Index=@n3) {materialFunction=&GetWaveData}
  n21 = CollectionParameter() {collection=&MPC_Water, parameterName=k3}
  n22 = MaterialFunctionCall(Float4=StaticSwitch(True=@n20.Params, False=CollectionParameter() {collection=&MPC_Water, parameterName=w3}, Value=@n6)) {materialFunction=&BreakOutFloat4Components}
  n23 = MaterialFunctionCall(
   Z Only Offset=@n1
   2D Direction Vector=StaticSwitch(True=@n20.Direction, False=ComponentMask(None=@n21) {g=true, r=true}, Value=@n6)
   Wave Length=@n22.R
   Amplitude=@n22.R
   Wave Steepness=@n22.R
   Pos=@n8
   wavesin 1=@n28.A
   wavesin 2=@n25.A
   wavecos 1=@n16.A
   wavecos 2=@n18.A
  ) {desc=3, materialFunction=&GerstnerWave_Deep_WithZTrace}
  MaterialExpressionMaterialFunctionCall_31 = MaterialFunctionCall(X=@n29.Wave Pos 2, Y=@n26.Wave Pos 2, Z=@n19.Wave Pos 2, A=@n23.Wave Pos 2) {materialFunction=&MakeFloat4}
  n24 = Custom(x=@MaterialExpressionMaterialFunctionCall_31.Result) {
   code="sin(x);"
   description="Sin(x)"
   outputType=CMOT_Float4
   inputs [{inputName=x, input {expression=&self:MaterialExpressionMaterialFunctionCall_31}}]
  }
  n25 = MaterialFunctionCall(Float4=@n24) {materialFunction=&BreakOutFloat4Components}
  n26 = MaterialFunctionCall(
   Z Only Offset=@n1
   2D Direction Vector=StaticSwitch(True=@n9.Direction, False=ComponentMask(None=@n10) {g=true, r=true}, Value=@n6)
   Wave Length=@n11.R
   Amplitude=@n11.R
   Wave Steepness=@n11.R
   Pos=@n8
   wavesin 1=@n28.G
   wavesin 2=@n25.G
   wavecos 1=@n16.G
   wavecos 2=@n18.G
  ) {desc=1, materialFunction=&GerstnerWave_Deep_WithZTrace}
  MaterialExpressionMaterialFunctionCall_0 = MaterialFunctionCall(X=@n29.Wave Pos 1, Y=@n26.Wave Pos 1, Z=@n19.Wave Pos 1, A=@n23.Wave Pos 1) {materialFunction=&MakeFloat4}
  n27 = Custom(x=@MaterialExpressionMaterialFunctionCall_0.Result) {
   code="sin(x);"
   description="Sin(x)"
   outputType=CMOT_Float4
   inputs [{inputName=x, input {expression=&self:MaterialExpressionMaterialFunctionCall_0}}]
  }
  n28 = MaterialFunctionCall(Float4=@n27) {materialFunction=&BreakOutFloat4Components}
  n29 = MaterialFunctionCall(
   Z Only Offset=@n1
   2D Direction Vector=StaticSwitch(True=@n4.Direction, False=ComponentMask(None=@n5) {g=true, r=true}, Value=@n6)
   Wave Length=@n7.R
   Amplitude=@n7.R
   Wave Steepness=@n7.R
   Pos=@n8
   wavesin 1=@n28.R
   wavesin 2=@n25.R
   wavecos 1=@n16.R
   wavecos 2=@n18.R
  ) {desc=0, materialFunction=&GerstnerWave_Deep_WithZTrace}
  FunctionOutput(None=Add(A=Add(A=@n29.WPO, B=@n26.WPO), B=Add(A=@n19.WPO, B=@n23.WPO))) {outputName=WPO, sortPriority=1}
  n30 = ComponentMask(None=Add(A=Add(A=@n29.Normal, B=@n26.Normal), B=Add(A=@n19.Normal, B=@n23.Normal))) {g=true, r=true}
  n31 = FunctionInput(Preview=StaticBool()) {
   bUsePreviewValueAsDefault=true
   description="If this is true, the first 4 waves will have their normal computed on the vertex shader."
   inputName=Normals use VS Interpolation
   inputType=FunctionInput_StaticBool
   sortPriority=11
  }
  n32 = AppendVector(A=StaticSwitch(True=VertexInterpolator(VS=@n30).PS, False=@n30, Value=@n31), B=Constant() {r=1})
  n33 = Constant3Vector() {constant=<rgba 0 0 1 0>}
  n34 = QualitySwitch(Default=@n32, Low=@n33, High=@n32, Medium=@n32)
  FunctionOutput(None=ShadingPathSwitch(Default=@n34, Deferred=@n34, Forward=@n33, Mobile=@n33)) {outputName=Normal}
 ]}
