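# Water material function that sums sixteen Gerstner waves for a water body into a normal and world position offset.
# It chains two six-wave calls and one four-wave call at different index offsets, keyed by position and WaterBodyIndex.
# Normal XY components are added and Z is re-derived, while the distance mask comes from the second six-wave call.
# Static switches allow Z-only displacement and vertex-interpolated normals.
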MaterialFunction /FortWater/Materials/Functions/MF_Gerstner_16_Waves
graph {nodes [
  n1 = FunctionInput(Preview=StaticBool()) {bUsePreviewValueAsDefault=true, inputName=Z Offset Only, inputType=FunctionInput_StaticBool}
  n2 = FunctionInput(Preview=WorldPosition() {worldPositionShaderOffset=WPT_ExcludeAllShaderOffsets}.XYZ) {bUsePreviewValueAsDefault=true, inputName=Position}
  n3 = FunctionInput(Preview=ScalarParameter() {group=Waves, parameterName=WaterBodyIndex}) {bUsePreviewValueAsDefault=true, inputName=WaterBodyIndex, inputType=FunctionInput_Scalar, sortPriority=5}
  n4 = FunctionInput(Preview=StaticBool()) {
   bUsePreviewValueAsDefault=true
   description="If this is true, the first 4 waves will have their normal computed on the vertex shader."
   inputName=Normals use VS Interpolation
   inputType=FunctionInput_StaticBool
   sortPriority=11
  }
  n5 = FunctionInput(Preview=StaticBool()) {
   bUsePreviewValueAsDefault=true
   description="If Normals use VS Interp is checked,  this allows forcing all 6 waves to use the interpolation."
   inputName=All Normals VS
   inputType=FunctionInput_StaticBool
   sortPriority=12
  }
  n6 = MaterialFunctionCall!(Z Offset Only=@n1, Position=@n2, WaterBodyIndex=@n3, Index Offset, Normals use VS Interpolation=@n4, All Normals VS=@n5) {materialFunction=&/MF_Gerstner_6_Waves_Multi}
  n7 = MaterialFunctionCall(Z Offset Only=@n1, Position=@n2, WaterBodyIndex=@n3, Index Offset=Constant() {r=5}, Normals use VS Interpolation=@n4, All Normals VS=@n5) {materialFunction=&MF_Gerstner_6_Waves_Multi}
  n8 = Add(A=ComponentMask(None=@n6.Normal) {g=true, r=true}, B=ComponentMask(None=@n7.Normal) {g=true, r=true})
  n9 = MaterialFunctionCall!(Z Offset Only=@n1, Position=@n2, WaterBodyIndex=@n3, Index Offset=Constant() {r=11}, Normals use VS Interpolation=@n4) {materialFunction=&/MF_Gerstner_4_Waves}
  n10 = ComponentMask(None=@n9.Normal) {g=true, r=true}
  FunctionOutput(None=DeriveNormalZ(InXY=Add(A=@n8, B=@n10))) {outputName=Normal}
  FunctionOutput(None=Add(A=Add(A=@n6.WPO, B=@n7.WPO), B=@n9.WPO)) {outputName=WPO, sortPriority=1}
  FunctionOutput(None=@n7.Distance Mask) {outputName=Distance Mask, sortPriority=2}
 ]}
