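# Material function that computes the specular value for water body shading, with time-of-day variation.
# Input specular fades in from a shallow value by water depth, optionally multiplied by a Fresnel-like boost for top-down views, while a basic switch passes the input through.
# Scalar_Time_of_Day blends the result as day value, factor-scaled night value, and a shared in-between value for morning and evening.
# A depth-factor multiply is left unconnected.
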MaterialFunction /FortWater/Materials/Functions/MF_WaterBodySpecular
graph {nodes [
  n1 = FunctionInput() {bUsePreviewValueAsDefault=true, inputName=InSpecular, inputType=FunctionInput_Scalar, previewValue=<0.2 0 0 1>}
  n2 = FunctionInput(Preview=Constant() {r=200}) {bUsePreviewValueAsDefault=true, inputName=WaterDepth, inputType=FunctionInput_Scalar, sortPriority=1}
  n3 = ScalarParameter() {defaultValue=50, group=Water Shading Model, parameterName=Specular_DepthFadeDistance}
  n4 = LinearInterpolate(
   A=ScalarParameter() {group=Water Shading Model, parameterName=Specular_Shallow}
   B=@n1
   Alpha=Saturate(None=Divide(A=@n2, B=@n3) {constB=50})
  )
  n5 = Saturate(None=DotProduct(A=VertexNormalWS(), B=CameraVectorWS()))
  n6 = ScalarParameter() {defaultValue=6, group=Water Shading Model, parameterName=Specular_TopDownBoost_Exponent}
  n7 = Add(A=Multiply(
    A=Power(Base=@n5, Exp=@n6) {constExponent=8}
    B=ScalarParameter() {defaultValue=20, group=Water Shading Model, parameterName=Specular_TopDownBoost_Intensity}
   ))
  n8 = StaticSwitchParameter(
   True=@n1
   False=StaticSwitchParameter(True=Multiply(A=@n4, B=@n7), False=@n4) {group=Water Shading Model, parameterName=UseTopDownSpecularBoost}
  ) {group=Water Shading Model, parameterName=UseBasicSpecular}
  n9 = Multiply(A=@n8, B=ScalarParameter() {defaultValue=1, group=Water Shading Model, parameterName=SpecularFactor_Night})
  n10 = Subtract(A=@n8, B=Divide(A=Subtract(A=@n8, B=@n9)))
  FunctionOutput(None=MaterialFunctionCall(Value_Morning=@n10, Value_Day=@n8, Value_Evening=@n10, Value_Night=@n9) {materialFunction=&Scalar_Time_of_Day}.Result)
  Multiply(A=@n2, B=ScalarParameter() {defaultValue=0.003, group=Water Shading Model, parameterName=Specular_DepthFactor}) {constB=0.003}
 ]}
