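# Material function that adds large-scale water color variation by blending input scattering and absorption toward parameter colors inside world-space regions.
# Sphere masks on rotated, scaled world XY define the regions, with static switches for mirroring or manual two-mask placement.
# Noise breaks up the mask, an optional transition mask limits it, and time of day scales and hue-shifts primary scattering.
# Outputs blended Scattering, Absorption and a packed Mask, behind an enable switch.
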MaterialFunction /FortWater/Materials/Functions/MF_WaterMacroColorVariation
graph {nodes [
  Constant() {r=-1.6}
  Constant() {r=1}
  Constant() {r=0.8}
  Constant() {r=-0.8}
  n1 = FunctionInput() {inputName=InAbsorption}
  n2 = FunctionInput() {inputName=InScattering}
  n3 = ScalarParameter() {group=MacroColorVariation, parameterName=TOD_HueShift}
  n4 = VectorParameter() {group=MacroColorVariation, parameterName=TOD_ScatteringFactors, channelNames {a=Night, b=Evening, g=Day, r=Morning}}
  n5 = MaterialFunctionCall(Value_Morning=@n4.Morning, Value_Day=@n4.Morning, Value_Evening=@n4.Morning, Value_Night=@n4.Morning) {materialFunction=&Scalar_Time_of_Day}
  n6 = OneMinus(None=@n5.Result)
  n7 = VectorParameter() {group=MacroColorVariation, parameterName=MacroColorVariation_Scattering, defaultValue=<rgba 1 1 1 1>}
  n8 = Multiply(
   A=MaterialFunctionCall(Hue Shift Percentage=Multiply(A=@n3, B=@n6), Texture=Multiply(A=@n5.Result, B=@n7.RGB)) {materialFunction=&HueShift}.Result
   B=Divide(A=@n7.A) {constB=100}
  )
  n9 = VectorParameter() {group=MacroColorVariation, parameterName=MacroColorVariation_Scattering_Mirrored, defaultValue=<rgba 1 1 1 1>}
  n10 = Rotator(
   Coordinate=WorldPosition().XY
   Time=StaticSwitchParameter(True=Constant(), False=ScalarParameter() {group=MacroColorVariation, parameterName=MacroColorVariation_Rotation}) {group=MacroColorVariation, parameterName=ManualMacroWaterColorVariation}
  )
  n11 = MaterialFunctionCall(Float2=@n10) {materialFunction=&BreakOutFloat2Components}
  n12 = MaterialFunctionCall(X=@n11.R, Y=Abs(None=@n11.G)) {materialFunction=&MakeFloat2}
  n13 = AppendVector(
   A=ScalarParameter() {defaultValue=1, group=MacroColorVariation, parameterName=MacroColorVariation_ScaleX}
   B=ScalarParameter() {defaultValue=1, group=MacroColorVariation, parameterName=MacroColorVariation_ScaleY}
  )
  n14 = Saturate(None=ComponentMask(None=@n10) {g=true})
  n15 = LinearInterpolate(
   A=@n13
   B=AppendVector(
    A=ScalarParameter() {defaultValue=1, group=MacroColorVariation, parameterName=MacroColorVariation_ScaleX_Mirrored}
    B=ScalarParameter() {defaultValue=1, group=MacroColorVariation, parameterName=MacroColorVariation_ScaleY_Mirrored}
   )
   Alpha=@n14
  )
  n16 = Multiply(
   A=StaticSwitchParameter(
    True=StaticSwitchParameter(True=@n10, False=@n12.Result) {group=MacroColorVariation, parameterName=ManualMacroWaterColorVariation}
    False=@n10
   ) {group=MacroColorVariation, parameterName=MirrorMacroWaterColorVariation}
   B=StaticSwitchParameter(True=StaticSwitchParameter(True=@n13, False=@n15) {group=MacroColorVariation, parameterName=ManualMacroWaterColorVariation}, False=@n13) {group=MacroColorVariation, parameterName=MirrorMacroWaterColorVariation}
  )
  n17 = ScalarParameter() {defaultValue=1.3, group=MacroColorVariation, parameterName=ManualMacroColorVariation_OffsetX}
  n18 = ScalarParameter() {defaultValue=-1.6, group=MacroColorVariation, parameterName=ManualMacroColorVariation_OffsetY}
  n19 = Constant() {r=50000}
  n20 = Add(A=@n16, B=Multiply(A=AppendVector(A=@n17, B=@n18), B=@n19))
  n21 = StaticSwitchParameter(
   True=Constant()
   False=VectorParameter() {group=MacroColorVariation, parameterName=MacroColorVariation_Position, defaultValue=<rgba 0 0 0 1>}.RGB
  ) {group=MacroColorVariation, parameterName=ManualMacroWaterColorVariation}
  n22 = MaterialFunctionCall(Float2=@n21) {materialFunction=&BreakOutFloat2Components}
  n23 = Abs(None=@n22.G)
  n24 = ComponentMask(None=StaticSwitchParameter(
    True=MaterialFunctionCall(X=@n22.R, Y=@n23) {materialFunction=&MakeFloat2}.Result
    False=@n21
   ) {group=MacroColorVariation, parameterName=MirrorMacroWaterColorVariation}) {g=true, r=true}
  n25 = SmoothStep(Value=SphereMask(
    A=StaticSwitchParameter(True=@n20, False=@n16) {group=MacroColorVariation, parameterName=ManualMacroWaterColorVariation}
    B=@n24
    Radius=ScalarParameter() {defaultValue=50000, group=MacroColorVariation, parameterName=MacroColorVariation_Radius}
    Hardness=ScalarParameter() {group=MacroColorVariation, parameterName=MacroColorVariation_Hardness}
   ))
  n26 = StaticSwitchParameter(True=OneMinus(None=Saturate(None=@n25)), False=@n14) {group=MacroColorVariation, parameterName=ManualMacroWaterColorVariation}
  n27 = LinearInterpolate(A=@n8, B=Multiply(A=@n9.RGB, B=Divide(A=@n9.A) {constB=100}), Alpha=@n26)
  n28 = VectorParameter() {group=MacroColorVariation, parameterName=MacroColorVariation_Absorption, defaultValue=<rgba 1 1 1 1>}
  n29 = Multiply(A=@n28.RGB, B=@n28.RGB)
  n30 = VectorParameter() {group=MacroColorVariation, parameterName=MacroColorVariation_Absorption_Mirrored, defaultValue=<rgba 1 1 1 1>}
  n31 = LinearInterpolate(A=@n29, B=Multiply(A=@n30.RGB, B=@n30.RGB), Alpha=@n26)
  n32 = OneMinus(None=FunctionInput() {bUsePreviewValueAsDefault=true, inputName=InTransitionMask, inputType=FunctionInput_Scalar})
  n33 = FunctionInput() {bUsePreviewValueAsDefault=true, inputName=InNoise, inputType=FunctionInput_Scalar, previewValue=<1 0 0 1>}
  n34 = FunctionInput(Preview=Constant() {r=1}) {bUsePreviewValueAsDefault=true, inputName=InOceanExtensionMask, inputType=FunctionInput_Scalar, previewValue=<1 0 0 1>}
  n35 = Multiply(
   A=AppendVector(
    A=ScalarParameter() {group=MacroColorVariation, parameterName=ManualMacroColorVariation_OffsetX_B}
    B=ScalarParameter() {defaultValue=2, group=MacroColorVariation, parameterName=ManualMacroColorVariation_OffsetY_B}
   )
   B=@n19
  )
  n36 = SmoothStep(Value=SphereMask(
    A=Add(A=@n16, B=@n35)
    B=@n24
    Radius=ScalarParameter() {defaultValue=200000, group=MacroColorVariation, parameterName=MacroColorVariation_Radius_B}
    Hardness=ScalarParameter() {group=MacroColorVariation, parameterName=MacroColorVariation_Hardness_B}
   ))
  n37 = StaticSwitchParameter(True=Multiply(A=Saturate(None=@n34), B=Max(A=@n25, B=@n36)), False=@n25) {group=MacroColorVariation, parameterName=ManualMacroWaterColorVariation}
  n38 = OneMinus(None=@n37)
  n39 = LinearInterpolate(A=@n37, B=Multiply(A=@n37, B=@n33), Alpha=Frac(None=Power(Base=@n38) {constExponent=4}))
  n40 = MaterialFunctionCall!(
   Scattering A=@n2
   Scattering B=StaticSwitchParameter(True=@n27, False=StaticSwitchParameter(True=@n27, False=@n8) {group=MacroColorVariation, parameterName=ManualMacroWaterColorVariation}) {group=MacroColorVariation, parameterName=MirrorMacroWaterColorVariation}
   Absorption A=@n1
   Absorption B=StaticSwitchParameter(True=@n31, False=StaticSwitchParameter(True=@n31, False=@n29) {group=MacroColorVariation, parameterName=ManualMacroWaterColorVariation}) {group=MacroColorVariation, parameterName=MirrorMacroWaterColorVariation}
   Alpha=StaticSwitchParameter(True=Multiply(A=@n32, B=@n39), False=@n39) {parameterName=UseMacroWaterColorVariation-TransitionMask}
  ) {materialFunction=&/BlendWaterShadingValues}
  FunctionOutput(None=StaticSwitchParameter(True=@n40.Scattering, False=@n2) {parameterName=UseMacroWaterColorVariation}) {outputName=Scattering}
  FunctionOutput(None=StaticSwitchParameter(True=@n40.Absorption, False=@n1) {parameterName=UseMacroWaterColorVariation}) {outputName=Absorption, sortPriority=1}
  n41 = ComponentMask(None=@n39) {r=true}
  n42 = Multiply(A=@n41, B=@n26)
  n43 = Multiply(A=@n41, B=OneMinus(None=@n26))
  FunctionOutput(None=StaticSwitchParameter(
    True=MaterialFunctionCall(X=@n41, Y=@n42, Z=@n43) {materialFunction=&MakeFloat3}.Result
    False=Constant3Vector()
   ) {parameterName=UseMacroWaterColorVariation}) {
   description="R contains mask that may or may not be mirrored, G contains positve mask if mirrored and B contains negative mask if mirrored"
   outputName=Mask
   sortPriority=5
  }
 ]}
