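# MF_Water_Styx is a material function that applies the glowing Styx look, in green or red variants, to water material attributes, absorption and scattering.
# Emissive combines two parallax, normal-distorted panning noise layers, depth-based edge glow, detail glow and macro noise, with distance fading and inverse eye-adaptation scaling.
# A world-position or water-velocity mask drives height-lerped transitions to normal water, with optional large-noise blending.
# Shading-path and quality switches substitute cheaper variants on mobile and lower tiers.
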MaterialFunction /FortWater/Materials/Functions/MF_Water_Styx
graph {nodes [
  n1 = Multiply(A=OneMinus(None=EyeAdaptation().EyeAdaptation), B=ScalarParameter() {defaultValue=2, parameterName=EyeadaptionInfluence})
  Abs(None=@n1)
  n2 = Divide(A=NamedRerouteUsage(), B=ScalarParameter() {defaultValue=11018.2959, group=Styx_Absoprtion_Noise, parameterName=Styx_Absoprtion_NoiseSize})
  Add(A=Multiply(A=NamedRerouteUsage(), B=ScalarParameter() {group=Styx_Absoprtion_Noise, parameterName=Styx_Absoprtion_Noise_UVdistort}), B=@n2)
  n3 = WorldPosition()
  Add(
   A=@n3.XY
   B=Multiply(A=NamedRerouteUsage(), B=ScalarParameter() {defaultValue=500, group=Styx_Absoprtion_Noise, parameterName=Styx_UVDistort_ProcWave})
  )
  n4 = VectorParameter() {group=Styx, parameterName=Styx_WaveDir_1, defaultValue=<rgba 1 -0.3 0 1>}
  n5 = Sine(None=Divide(A=DotProduct(A=@n3.XY, B=@n4.RGB), B=ScalarParameter() {defaultValue=5000, group=Styx, parameterName=Styx_ProcWaveLength_1}))
  n6 = VectorParameter() {group=Styx, parameterName=Styx_WaveDir_2, defaultValue=<rgba 0.3 1 0 1>}
  n7 = Sine(None=Divide(A=DotProduct(A=@n3.XY, B=@n6.RGB), B=ScalarParameter() {defaultValue=6970.8472, group=Styx, parameterName=Styx_ProcWaveLength_2}))
  n8 = Add(A=Divide(A=Add(A=@n5)), B=Divide(A=Add(A=@n7)))
  n9 = VectorParameter() {group=Styx, parameterName=Styx_WaveDir_3, defaultValue=<rgba 1 1 0 1>}
  n10 = Sine(None=Divide(A=DotProduct(A=@n3.XY, B=@n9.RGB), B=ScalarParameter() {defaultValue=5500, group=Styx, parameterName=Styx_ProcWaveLength_4}))
  Add(A=@n8, B=Divide(A=Add(A=@n10)))
  Constant3Vector()
  Constant()
  n11 = WorldPosition()
  n12 = VectorParameter() {group=Styx, parameterName=Styx_toWater_WP_Transition_Position, defaultValue=<rgba 0 0 0 1>}
  Distance(A=@n11.XYZ, B=@n12.RGB)
  FunctionInput() {bUsePreviewValueAsDefault=true, inputName=SimWaterNormal, sortPriority=22, previewValue=<0 0 1 1>}
  n13 = FunctionInput() {inputName=DefaultMaterial, inputType=FunctionInput_MaterialAttributes}
  n14 = FunctionInput() {inputName=Material_NoFoam, inputType=FunctionInput_MaterialAttributes, sortPriority=1}
  n15 = Subtract(A=@n11.XYZ, B=@n12.RGB)
  n16 = Normalize(VectorInput=VectorParameter() {group=Styx, parameterName=Styx_toWater_WP_Transition_Direction, defaultValue=<rgba 1 0 0 1>}.RGB)
  n17 = Saturate(None=Divide(A=DotProduct(A=@n15, B=@n16), B=ScalarParameter() {defaultValue=50000, group=Styx, parameterName=Styx_toWater_WP_Transition_Distance}))
  n18 = MaterialFunctionCall!(Texture, Position) {materialFunction=&/WaterBodyData}
  n19 = Saturate(None=Divide(A=ComponentMask(None=@n18.VelocityAndLength) {b=true}, B=ScalarParameter() {defaultValue=48, parameterName=Velocity for Scattering}))
  n20 = FunctionInput(Preview=StaticSwitch(
    True=StaticSwitchParameter(True=@n17, False=@n19) {group=Styx, parameterName=Is Mobile Lake To Ocean Transiiton}
    False=@n17
    Value=NamedRerouteUsage()
   )) {bUsePreviewValueAsDefault=true, inputName=TransitionMask, inputType=FunctionInput_Scalar, sortPriority=50, previewValue=<1 0 0 1>}
  n21 = GetMaterialAttributes(None=StaticSwitchParameter(True=BlendMaterialAttributes(A=@n13, B=@n14, Alpha=@n20), False=@n14) {group=Styx, parameterName=EnableStyxOceanTransition}) {attributeGetTypes [#1]}
  n22 = ScalarParameter() {defaultValue=4400.8735, group=Styx_FloatingLayer_Shared, parameterName=StyxNoiseSize2}
  n23 = Divide(A=NamedRerouteUsage(), B=@n22)
  n24 = ScalarParameter() {defaultValue=0.005, group=Styx_FloatingLayer_Shared, parameterName=StyxNoise_Speed2}
  n25 = TextureSample(UVs=Add(A=NamedRerouteUsage(), B=Panner(Coordinate=@n23, Speed=@n24))) {samplerSource=SSM_Wrap_WorldGroupSettings, samplerType=SAMPLERTYPE_Masks, texture=/FortWater/Textures/Styx/StyxWater_Mask1}
  n26 = Multiply(
   A=StaticSwitchParameter(
    True=VectorParameter() {group=Styx_FloatingLayer_2, parameterName=StyxEmissive2_Red, defaultValue=<rgba 5 2.1297 2.6805 1>}.RGB
    False=VectorParameter() {group=Styx_FloatingLayer_2, parameterName=StyxEmissive2, defaultValue=<rgba 0.2146 1.3 0.5683 1>}.RGB
   ) {group=Styx, parameterName=IsStyxRed}
   B=Power(Base=@n25.R, Exp=ScalarParameter() {defaultValue=4, group=Styx_FloatingLayer_2, parameterName=StyxNoise_Contrast2})
  )
  n27 = VectorParameter() {group=Styx_FloatingLayer_1, parameterName=StyxEmissive_Red, defaultValue=<rgba 5 0.1396 0 1>}
  n28 = VectorParameter() {group=Styx_FloatingLayer_1, parameterName=StyxEmissive, defaultValue=<rgba 0.0799 0.5 0.1709 1>}
  n29 = NamedRerouteUsage()
  n30 = Panner(Coordinate=Divide(A=NamedRerouteUsage(), B=@n22), Speed=@n24)
  n31 = Power(
   Base=TextureSample(UVs=Add(A=@n29, B=@n30)) {samplerSource=SSM_Wrap_WorldGroupSettings, samplerType=SAMPLERTYPE_Masks, texture=&StyxWater_Mask1}.R
   Exp=LinearInterpolate(
    A=ScalarParameter() {defaultValue=1.3788, group=Styx_FloatingLayer_1, parameterName=StyxNoise_Contrast}
    B=ScalarParameter() {defaultValue=0.4976, group=Styx_FloatingLayer_1, parameterName=StyxNoise_Contrast_far}
    Alpha=NamedRerouteUsage()
   )
  )
  n32 = LinearInterpolate(
   A=@n26
   B=Multiply(A=StaticSwitchParameter(True=@n27.RGB, False=@n28.RGB) {group=Styx, parameterName=IsStyxRed}, B=@n31)
   Alpha=Saturate(None=@n31)
  )
  n33 = Multiply(
   A=ShadingPathSwitch(Default=QualitySwitch(Default=@n32, Low=@n26, High=@n26, Medium=@n26), Mobile=@n26) {desc=Migrated from FeatureLevelSwitch}
   B=NamedRerouteUsage()
  )
  n34 = Multiply(A=@n33, B=ScalarParameter() {defaultValue=0.22, group=Styx, parameterName=Styx_Distance_Fade_Brightness})
  n35 = Subtract(A=PixelDepth().R, B=ScalarParameter() {defaultValue=2000, group=Styx, parameterName=Styx_Distance_Fade_Offset})
  n36 = ScalarParameter() {defaultValue=15553.8877, group=Styx, parameterName=Styx_Distance_Fade_Distance}
  n37 = NamedRerouteDeclaration(None=Saturate(None=Divide(A=@n35, B=@n36))) {name=Distance Fade, nodeColor=<rgba 0 0.5312 1 1>}
  n38 = NamedRerouteUsage()
  n39 = StaticSwitchParameter(
   True=VectorParameter() {group=Styx_Detail, parameterName=StyxEmissive_DetailGlow_Red, defaultValue=<rgba 8 0 1.161 1>}.RGB
   False=VectorParameter() {group=Styx_Detail, parameterName=StyxEmissive_DetailGlow, defaultValue=<rgba 1.976 8 1.0055 1>}.RGB
  ) {group=Styx, parameterName=IsStyxRed}
  n40 = Desaturation(None=@n33)
  n41 = ScalarParameter() {defaultValue=0.02, group=Styx_Detail, parameterName=Styx_Detail_Noise_MinValue}
  n42 = NamedRerouteUsage()
  n43 = Subtract(
   A=FunctionInput() {inputName=Depth, inputType=FunctionInput_Scalar, sortPriority=5}
   B=ScalarParameter() {defaultValue=-24.0032, group=Styx_Edge, parameterName=Styx_EdgeGlow_Offset}
  )
  n44 = ScalarParameter() {defaultValue=243.9534, group=Styx_Edge, parameterName=Styx_EdgeGlow_Depth}
  n45 = OneMinus(None=Saturate(None=Divide(A=@n43, B=@n44)))
  n46 = Power(Base=@n45, Exp=ScalarParameter() {defaultValue=3.0739, group=Styx_Edge, parameterName=Styx_EdgeGlow_Falloff})
  n47 = ScalarParameter() {defaultValue=5, group=Styx_Edge, parameterName=Styx_EdgeGlow_FalloffFade}
  n48 = OneMinus(None=Saturate(None=Power(Base=@n45, Exp=@n47)))
  n49 = Multiply(
   A=@n46
   B=MaterialFunctionCall(
    Transition Phase=@n48
    Height Texture=NamedRerouteUsage()
    Contrast=ScalarParameter() {defaultValue=-0.001, group=Styx_Edge, parameterName=Styx_EdgeGlow_FalloffFadeContrast}
   ) {materialFunction=&HeightLerp}.Alpha
  )
  n50 = Multiply(A=Multiply(A=@n46, B=@n48), B=ScalarParameter() {defaultValue=0.2, group=Styx_Edge, parameterName=Styx_EdgeGlow_LowDetail})
  n51 = ShadingPathSwitch(Default=QualitySwitch(Default=Multiply(A=@n42, B=@n49), Low=@n50, High=@n50, Medium=@n50), Mobile=@n50) {desc=Migrated from FeatureLevelSwitch}
  n52 = NamedRerouteUsage()
  n53 = Multiply(A=Add(A=Max(A=@n40, B=@n41) {constB=0.1}, B=@n51), B=@n52)
  n54 = Add(A=Add(A=LinearInterpolate(A=@n33, B=@n34, Alpha=@n37), B=NamedRerouteUsage()), B=Multiply(A=OneMinus(None=@n38), B=Multiply(A=@n39, B=@n53)))
  n55 = VectorParameter() {group=Water Shading Model, parameterName=Mobile_Absorption_A_Styx_Red, defaultValue=<rgba 1 0.2003 0.1828 1>}
  n56 = VectorParameter() {group=Water Shading Model, parameterName=Mobile_Absorption_A_Styx, defaultValue=<rgba 0.0063 0.1406 0.0843 1>}
  n57 = VectorParameter() {group=Water Shading Model, parameterName=Mobile_Absorption_B_Styx_Red, defaultValue=<rgba 1.7 0.2282 0.1944 1>}
  n58 = VectorParameter() {group=Water Shading Model, parameterName=Mobile_Absorption_B_Styx, defaultValue=<rgba 0.2412 1.7 0.895 1>}
  n59 = Add(
   A=@n54
   B=LinearInterpolate(
    A=StaticSwitchParameter(True=@n55.RGB, False=@n56.RGB) {defaultValue=true, group=Styx, parameterName=IsStyxRed}
    B=StaticSwitchParameter(True=@n57.RGB, False=@n58.RGB) {group=Styx, parameterName=IsStyxRed}
    Alpha=NamedRerouteUsage()
   )
  )
  n60 = Add(A=@n54, B=NamedRerouteUsage())
  n61 = NamedRerouteUsage()
  n62 = Multiply(
   A=Multiply(A=StaticSwitch(True=@n59, False=@n60, Value=@n61), B=ScalarParameter() {defaultValue=0.5, group=Styx, parameterName=Styx_Final_Brightness})
   B=Saturate(None=TwoSidedSign())
  )
  n63 = FunctionInput() {bUsePreviewValueAsDefault=true, inputName=Mobile Reflections, sortPriority=556}
  n64 = VectorParameter() {parameterName=Styx_Reflection_Mult, defaultValue=<rgba 0.3594 0.3594 0.3594 1>}
  n65 = SceneDepthWithoutWater!(UVs)
  n66 = PixelDepth()
  n67 = Divide(
   A=NamedRerouteDeclaration(None=Subtract(A=@n65.R, B=@n66.R)) {name=Depth, nodeColor=<rgba 0.2891 1 0 1>}
   B=ScalarParameter() {defaultValue=19.7405, group=Styx, parameterName=Styx_DepthFade}
  )
  n68 = LinearInterpolate(
   A=Multiply(A=Saturate(None=@n67), B=@n62)
   B=@n62
   Alpha=Saturate(None=Subtract(A=@n66.R) {constB=25000, desc=Fix for Black border at Minimap Distance})
  )
  n69 = StaticSwitch(True=Add(A=@n62, B=Multiply(A=@n63, B=@n64.RGB)), False=@n68, Value=NamedRerouteUsage())
  n70 = Multiply(
   A=OneMinus(None=FunctionInput() {bUsePreviewValueAsDefault=true, inputName=SimErasure, inputType=FunctionInput_Scalar, sortPriority=22})
   B=Divide(A=@n69, B=EyeAdaptation().EyeAdaptation)
  )
  n71 = MaterialFunctionCall(
   Transition Phase=LinearInterpolate(
    A=ScalarParameter() {defaultValue=1, group=Styx, parameterName=Styx_toWater_Transition_High}
    B=ScalarParameter() {group=Styx, parameterName=Styx_toWater_Transition_Low}
    Alpha=@n20
   )
   Height Texture=NamedRerouteUsage()
   Contrast=ScalarParameter() {group=Styx, parameterName=Styx_toWater_Transition_Contrast}
  ) {materialFunction=&HeightLerp}
  n72 = OneMinus(None=NamedRerouteDeclaration(None=@n71.Lerp Alpha No Contrast) {name=TransitionMask, nodeColor=<rgba 0.2656 1 0 1>})
  n73 = StaticSwitchParameter(
   True=StaticSwitch(True=LinearInterpolate(A=@n21.EmissiveColor, B=@n70, Alpha=@n72), False=Multiply(A=@n72, B=@n70), Value=NamedRerouteUsage())
   False=@n70
  ) {group=Styx, parameterName=UseStyxTransition}
  n74 = SetMaterialAttributes!(MaterialAttributes=@n21.MaterialAttributes, EmissiveColor=@n73) {attributeSetTypes [#1]}
  n75 = BlendMaterialAttributes(A=@n74, B=@n21.MaterialAttributes, Alpha=NamedRerouteUsage())
  FunctionOutput(None=StaticSwitchParameter(True=StaticSwitchParameter(True=@n75, False=@n74) {parameterName=UseStyxBlendWithNormalWater}, False=@n13) {parameterName=UseStyxShading})
  n76 = Constant() {r=1}
  FunctionOutput(None=@n76) {outputName=ColorScale, sortPriority=5}
  n77 = VectorParameter() {group=Water Shading Model, parameterName=Scattering_A_Styx_Red, defaultValue=<rgba 0.276 0.0783 0.0563 1.3396>}
  n78 = Divide(A=@n77.A) {constB=100}
  n79 = VectorParameter() {group=Water Shading Model, parameterName=Scattering_A_Styx, defaultValue=<rgba 0.0531 0.1354 0.0203 1.3396>}
  n80 = Divide(A=@n79.A) {constB=100}
  n81 = VectorParameter() {group=Water Shading Model, parameterName=Scattering_B_Styx_Red, defaultValue=<rgba 0.25 0.0166 0 2.1458>}
  n82 = Divide(A=@n81.A) {constB=100}
  n83 = VectorParameter() {group=Water Shading Model, parameterName=Scattering_B_Styx, defaultValue=<rgba 0.0061 0.4948 0.0336 2.1458>}
  n84 = Divide(A=@n83.A) {constB=100}
  n85 = NamedRerouteUsage()
  n86 = LinearInterpolate(
   A=StaticSwitchParameter(True=Multiply(A=@n77.RGB, B=@n78), False=Multiply(A=@n79.RGB, B=@n80)) {group=Styx, parameterName=IsStyxRed}
   B=StaticSwitchParameter(True=Multiply(A=@n81.RGB, B=@n82), False=Multiply(A=@n83.RGB, B=@n84)) {group=Styx, parameterName=IsStyxRed}
   Alpha=@n85
  )
  n87 = VectorParameter() {group=Water Shading Model, parameterName=Absorption_A_Styx_Red, defaultValue=<rgba 0.5625 0.1093 0.0791 259.5518>}
  n88 = VectorParameter() {group=Water Shading Model, parameterName=Absorption_A_Styx, defaultValue=<rgba 0.3011 0.5208 0.3011 259.5518>}
  n89 = VectorParameter() {group=Water Shading Model, parameterName=Absorption_B_Styx_Red, defaultValue=<rgba 5 0.1317 0 259.5518>}
  n90 = VectorParameter() {group=Water Shading Model, parameterName=Absorption_B_Styx, defaultValue=<rgba 0.1191 0.5365 0.3541 259.5518>}
  n91 = LinearInterpolate(
   A=StaticSwitchParameter(True=Multiply(A=@n87.RGB, B=@n87.RGB), False=Multiply(A=@n88.RGB, B=@n88.RGB)) {group=Styx, parameterName=IsStyxRed}
   B=StaticSwitchParameter(True=Multiply(A=@n89.RGB, B=@n89.RGB), False=Multiply(A=@n90.RGB, B=@n90.RGB)) {group=Styx, parameterName=IsStyxRed}
   Alpha=@n85
  )
  n92 = NamedRerouteUsage()
  n93 = NamedRerouteUsage()
  n94 = MaterialFunctionCall!(
   Scattering A=@n86
   Scattering B=FunctionInput() {bUsePreviewValueAsDefault=true, inputName=Scattering, sortPriority=52}
   Absorption A=@n91
   Absorption B=FunctionInput() {bUsePreviewValueAsDefault=true, inputName=Absorption, sortPriority=54}
   Alpha=StaticSwitchParameter(True=Add(A=@n92, B=@n93), False=@n92) {parameterName=UseStyxBlendWithNormalWater}
  ) {materialFunction=&/BlendWaterShadingValues}
  FunctionOutput(None=StaticSwitchParameter(True=@n94.Absorption, False=StaticSwitchParameter(True=@n94.Absorption, False=@n91) {group=Styx, parameterName=UseStyxTransition}) {parameterName=UseStyxBlendWithNormalWater}) {outputName=Absorbtion, sortPriority=5}
  FunctionOutput(None=StaticSwitchParameter(True=@n94.Scattering, False=StaticSwitchParameter(True=@n94.Scattering, False=@n86) {group=Styx, parameterName=UseStyxTransition}) {parameterName=UseStyxBlendWithNormalWater}) {outputName=Scatter, sortPriority=6}
  FunctionOutput(None=@n72) {outputName=TransitionMask, sortPriority=4545}
  MaterialFunctionCall(
   Hue Shift Percentage=ScalarParameter() {defaultValue=0.128, group=Styx, parameterName=StyxHueShift}
   Texture=Desaturation(None=@n69, Fraction=ScalarParameter() {defaultValue=0.2, group=Styx, parameterName=StyxSaturation})
  ) {materialFunction=&HueShift}
  n95 = NamedRerouteUsage()
  n96 = Divide(A=ScalarParameter() {defaultValue=80, group=Styx, parameterName=Styx_Detail_DepthFade}, B=NamedRerouteUsage())
  Multiply(A=@n52, B=Saturate(None=Divide(A=@n95, B=@n96))) {desc=Optimzed almost no difference}
  n97 = Divide(A=@n8)
  n98 = Divide(A=@n3.XY, B=ScalarParameter() {defaultValue=7000, parameterName=StyxSpheremask_size})
  n99 = Divide(A=Distance(A=Frac(None=@n98), B=Constant2Vector() {g=0.5, r=0.5})) {constB=1}
  n100 = Saturate(None=Add(A=OneMinus(None=@n99), B=ScalarParameter() {defaultValue=-0.75, parameterName=StyxSphereMask Intensity}))
  Multiply(A=@n97, B=@n100)
  n101 = ScalarParameter() {defaultValue=0.624, group=Styx_Absoprtion_Noise, parameterName=Styx_Absoprtion_Noise_MinValue}
  n102 = ScalarParameter() {group=Styx_Absoprtion_Noise, parameterName=Styx_Absoprtion_Noise_Contrast_LOW}
  n103 = ScalarParameter() {defaultValue=0.25, group=Styx_Absoprtion_Noise, parameterName=Styx_Absoprtion_Noise_Contrast_High}
  n104 = TextureSample(UVs=@n2) {
   samplerSource=SSM_Wrap_WorldGroupSettings
   samplerType=SAMPLERTYPE_Grayscale
   texture=/FortWater/Textures/Styx/StyxWater_DetailNoise_HighQuality_M
  }
  n105 = Add(A=@n97, B=@n100)
  n106 = Saturate(None=LinearInterpolate(
    A=ScalarParameter() {defaultValue=0.19, group=Styx_Absoprtion_Noise, parameterName=Styx_Absoprtion_Noise_B_Contrast_PS4_LOW}
    B=ScalarParameter() {defaultValue=0.05, group=Styx_Absoprtion_Noise, parameterName=Styx_Absoprtion_Noise_B_Contrast_PS4_High}
    Alpha=Saturate(None=@n105)
   ))
  n107 = NamedRerouteDeclaration(None=QualitySwitch(Default=@n104.G, High=@n106, Medium=@n106)) {name=MacroNoise_base, nodeColor=<rgba 1 0.8203 0 1>}
  n108 = ScalarParameter() {defaultValue=-0.5469, group=Styx_Absoprtion_Noise, parameterName=Styx_Absoprtion_Noise_B_Contrast_LOW}
  n109 = ScalarParameter() {defaultValue=0.5399, group=Styx_Absoprtion_Noise, parameterName=Styx_Absoprtion_Noise_B_Contrast_High}
  n110 = Add(A=Saturate(None=LinearInterpolate(A=@n102, B=@n103, Alpha=@n107)), B=Saturate(None=LinearInterpolate(A=@n108, B=@n109, Alpha=@n104.G)))
  n111 = Saturate(None=Multiply(A=LinearInterpolate(A=@n101, Alpha=@n25.G), B=@n110))
  n112 = Saturate(None=LinearInterpolate(
    A=ScalarParameter() {defaultValue=0.1799, group=Styx_Absoprtion_Noise, parameterName=Styx_Absoprtion_Noise_B_Contrast_Mobile_LOW}
    B=ScalarParameter() {defaultValue=-0.0749, group=Styx_Absoprtion_Noise, parameterName=Styx_Absoprtion_Noise_B_Contrast_Mobile_High}
    Alpha=@n104.G
   ))
  NamedRerouteDeclaration(None=ShadingPathSwitch(Default=QualitySwitch(Default=@n111, Low=@n112, High=@n106, Medium=@n106), Mobile=@n112) {desc=Migrated from FeatureLevelSwitch}) {name=MarcoNoiseMask, nodeColor=<rgba 0 1 0.8359 1>}
  n113 = WorldPosition() {worldPositionShaderOffset=WPT_ExcludeAllShaderOffsets}
  n114 = CameraVectorWS()
  n115 = ScalarParameter() {defaultValue=-20, group=Styx_FloatingLayer_1, parameterName=StyxNoise_DepthOffset}
  n116 = DotProduct(A=@n114, B=PixelNormalWS())
  n117 = Add(A=@n113.XY, B=Multiply(A=ComponentMask(None=@n114) {g=true, r=true}, B=Divide(A=@n115, B=@n116)))
  NamedRerouteDeclaration(None=ShadingPathSwitch(Default=QualitySwitch(Default=@n117, Low=@n113.XY, High=@n113.XY, Medium=@n113.XY), Mobile=@n113.XY) {desc=Migrated from FeatureLevelSwitch}) {name=UV_withParallax_A, nodeColor=<rgba 1 0.6563 0 1>}
  n118 = VectorParameter() {group=Styx_Absoprtion_Noise, parameterName=StyxEmissive_MacroNoise_A_Red, defaultValue=<rgba 0.2344 0.1139 0.1046 1>}
  n119 = VectorParameter() {group=Styx_Absoprtion_Noise, parameterName=StyxEmissive_MacroNoise_A, defaultValue=<rgba 0.0094 0.0313 0.1 1>}
  n120 = StaticSwitchParameter(True=@n118.RGB, False=@n119.RGB) {defaultValue=true, group=Styx, parameterName=IsStyxRed}
  n121 = VectorParameter() {group=Styx_Absoprtion_Noise, parameterName=StyxEmissive_MacroNoise_B_Red, defaultValue=<rgba 0.7604 0.273 0.2276 1>}
  n122 = VectorParameter() {group=Styx_Absoprtion_Noise, parameterName=StyxEmissive_MacroNoise_B, defaultValue=<rgba 0.5306 1 0.3012 1>}
  n123 = StaticSwitchParameter(True=@n121.RGB, False=@n122.RGB) {group=Styx, parameterName=IsStyxRed}
  n124 = NamedRerouteUsage()
  NamedRerouteDeclaration(None=Multiply(A=LinearInterpolate(A=@n120, B=@n123, Alpha=@n124), B=NamedRerouteUsage())) {name=Final_MarcorGlowVariation, nodeColor=<rgba 0 1 0.2969 1>}
  n125 = ScalarParameter() {defaultValue=-0.112, group=Styx_Detail, parameterName=Styx_Detail_Noise_Low}
  n126 = ScalarParameter() {defaultValue=1, group=Styx_Detail, parameterName=Styx_Detail_Noise_High}
  n127 = NamedRerouteUsage()
  n128 = Divide(A=NamedRerouteUsage(), B=ScalarParameter() {defaultValue=985.8951, group=Styx_Detail, parameterName=Styx_Detail_NoiseSize})
  n129 = TextureSample(UVs=Panner(Coordinate=Add(A=@n127, B=@n128), Speed=ScalarParameter() {defaultValue=0.02, group=Styx_Detail, parameterName=Styx_Detail_Noise_Speed})) {samplerSource=SSM_Wrap_WorldGroupSettings, samplerType=SAMPLERTYPE_Masks, texture=/FortWater/Textures/Styx/StyxWater_DetailNoise_M}
  NamedRerouteDeclaration(None=Saturate(None=LinearInterpolate(A=@n125, B=@n126, Alpha=@n129.R))) {name=DetailNoise, nodeColor=<rgba 1 0.0938 0 1>}
  n130 = ScalarParameter() {defaultValue=-1.4293, parameterName=Huge_Mask_contrast_low}
  n131 = ScalarParameter() {defaultValue=3.0992, parameterName=Huge_Mask_contrastHigh}
  n132 = TextureSample(UVs=Divide(A=WorldPosition().XY, B=ScalarParameter() {defaultValue=31889.2617, parameterName=Huge_Mask_SIze})) {
   samplerSource=SSM_Wrap_WorldGroupSettings
   samplerType=SAMPLERTYPE_Grayscale
   texture=&StyxWater_DetailNoise_HighQuality_M
  }
  NamedRerouteDeclaration(None=Saturate(None=LinearInterpolate(A=@n130, B=@n131, Alpha=@n132.R))) {name=StyxNormalWatrBlend, nodeColor=<rgba 0 1 0.25 1>}
  NamedRerouteDeclaration(None=FunctionInput(Preview=StaticBool()) {bUsePreviewValueAsDefault=true, inputName=Is Translucent Mat, inputType=FunctionInput_StaticBool, sortPriority=555}) {name=Is Translucent Mateiral, nodeColor=<rgba 0 1 0.1797 1>}
  n133 = WorldPosition() {worldPositionShaderOffset=WPT_ExcludeAllShaderOffsets}
  n134 = CameraVectorWS()
  n135 = ScalarParameter() {defaultValue=-50, group=Styx_FloatingLayer_2, parameterName=StyxNoise_DepthOffset2}
  n136 = DotProduct(A=@n134, B=PixelNormalWS())
  n137 = NamedRerouteDeclaration(None=@n136) {name=CameraDotNormal, nodeColor=<rgba 1 0 0.8437 1>}
  n138 = Add(A=@n133.XY, B=Multiply(A=ComponentMask(None=@n134) {g=true, r=true}, B=Divide(A=@n135, B=@n137)))
  NamedRerouteDeclaration(None=ShadingPathSwitch(Default=QualitySwitch(Default=@n138, Low=@n133.XY, High=@n133.XY, Medium=@n133.XY), Mobile=@n133.XY) {desc=Migrated from FeatureLevelSwitch}) {name=UV_withParallax_B, nodeColor=<rgba 0 0.4844 1 1>}
  n139 = ScalarParameter() {defaultValue=3221.0681, group=Styx_FloatingLayer_Shared, parameterName=StyxNoiseSize}
  n140 = ScalarParameter() {defaultValue=0.03, group=Styx_FloatingLayer_Shared, parameterName=StyxNoise_Speed}
  n141 = Panner(Coordinate=Divide(A=NamedRerouteUsage(), B=@n139), Speed=@n140)
  NamedRerouteDeclaration(None=ComponentMask(None=TextureSample(UVs=@n141) {samplerSource=SSM_Wrap_WorldGroupSettings, samplerType=SAMPLERTYPE_Normal, texture=/FortWater/Textures/Styx/StyxWater_N1}.RGB) {g=true, r=true}) {name=Distortion_NormalNoise_A, nodeColor=<rgba 0.3828 0 1 1>}
  n142 = Panner(Coordinate=Divide(A=NamedRerouteUsage(), B=@n139), Speed=@n140)
  NamedRerouteDeclaration(None=ComponentMask(None=TextureSample(UVs=@n142) {samplerSource=SSM_Wrap_WorldGroupSettings, samplerType=SAMPLERTYPE_Normal, texture=&StyxWater_N1}.RGB) {g=true, r=true}) {name=Distortion_NormalNoise_B, nodeColor=<rgba 0.1016 0 1 1>}
  n143 = ScalarParameter() {group=Styx_Edge, parameterName=EdgeGlow_Noise_Contrast_Low}
  n144 = ScalarParameter() {defaultValue=1, group=Styx_Edge, parameterName=EdgeGlow_Noise_Contrast_HIGH}
  n145 = NamedRerouteUsage()
  n146 = ScalarParameter() {defaultValue=5115.3262, group=Styx_Edge, parameterName=Styx_NoiseEdge_Size}
  n147 = NamedRerouteUsage()
  n148 = ScalarParameter() {defaultValue=0.1, group=Styx_Edge, parameterName=Styx_NoiseEdge_Distortion}
  n149 = TextureSample(UVs=Add(A=Divide(A=@n145, B=@n146), B=Multiply(A=@n147, B=@n148))) {samplerSource=SSM_Wrap_WorldGroupSettings, samplerType=SAMPLERTYPE_Masks, texture=&StyxWater_Mask1}
  NamedRerouteDeclaration(None=Saturate(None=LinearInterpolate(A=@n143, B=@n144, Alpha=@n149.G))) {name=NoisePerlin, nodeColor=<rgba 0.2188 0 1 1>}
  n150 = ScalarParameter() {defaultValue=0.02, group=Styx_FloatingLayer_Shared, parameterName=Styx_Distort_UV}
  n151 = Multiply(
   A=ComponentMask(None=PixelNormalWS()) {g=true, r=true}
   B=ScalarParameter() {defaultValue=0.5, group=Styx_FloatingLayer_Shared, parameterName=Styx_Distort_UV_PixelNormal}
  )
  n152 = Add(A=Multiply(A=NamedRerouteUsage(), B=@n150), B=@n151)
  n153 = Constant()
  NamedRerouteDeclaration(None=ShadingPathSwitch(Default=QualitySwitch(Default=@n152, Low=@n153, High=@n153, Medium=@n153), Mobile=@n153) {desc=Migrated from FeatureLevelSwitch}) {name=Distortion_Final_A, nodeColor=<rgba 0.2422 0 1 1>}
  n154 = Add(A=@n151, B=Multiply(A=NamedRerouteUsage(), B=@n150))
  NamedRerouteDeclaration(None=ShadingPathSwitch(Default=QualitySwitch(Default=@n154, Low=@n153, High=@n153, Medium=@n153), Mobile=@n153) {desc=Migrated from FeatureLevelSwitch}) {name=Distortion_Final_B, nodeColor=<rgba 1 0 0.8906 1>}
  NamedRerouteDeclaration(None=OneMinus(None=Saturate(None=@n46))) {name=EdgeGradient_A, nodeColor=<rgba 1 0.8906 0 1>}
  n155 = Multiply(
   A=Clamp(None=@n1) {clampMode=CMODE_ClampMin}
   B=VectorParameter() {group=Styx_Edge, parameterName=StyxEmissive_EdgeGlow_Red, defaultValue=<rgba 25 0 0 1>}.RGB
  ) {constB=2}
  n156 = VectorParameter() {group=Styx_Edge, parameterName=StyxEmissive_EdgeGlow, defaultValue=<rgba 0.4008 3.5456 1.7896 1>}
  NamedRerouteDeclaration(None=Multiply(A=StaticSwitchParameter(True=@n155, False=@n156.RGB) {group=Styx, parameterName=IsStyxRed}, B=@n51)) {name=Final_EdgeGlow, nodeColor=<rgba 0 1 0.0156 1>}
  NamedRerouteUsage()
  NamedRerouteUsage()
  OneMinus(None=Saturate(None=@n110))
  OneMinus(None=@n19)
  OneMinus(None=@n20)
  n157 = Saturate(None=LinearInterpolate(
    A=ScalarParameter() {defaultValue=-0.5, group=Styx, parameterName=Styx_toWater_Transition_PS4_High}
    B=ScalarParameter() {defaultValue=2, group=Styx, parameterName=Styx_toWater_Transition_PS4_Low}
    Alpha=@n20
   ))
  QualitySwitch(Default=@n71.Lerp Alpha No Contrast, High=@n157, Medium=@n157)
  Reroute(None=@n136)
  ScalarParameter() {defaultValue=50000, group=Styx, parameterName=Styx_toWater_WP_Transition_Offset}
  ScalarParameter() {defaultValue=1, parameterName=StyxSphereMask Intensity2}
  ShadingPathSwitch(Default=QualitySwitch(Default=@n68, Low=@n62, High=@n62, Medium=@n62), Mobile=@n62) {desc=Migrated from FeatureLevelSwitch}
  StaticSwitchParameter(
   True=LinearInterpolate(
    A=Constant3Vector() {constant=<rgba 1 1 1 0>}
    B=VectorParameter() {group=Styx_Edge, parameterName=Styx_ColorScale, defaultValue=<rgba 0 0 0 1>}.RGB
    Alpha=@n51
   ) {constA=1}
   False=@n76
  ) {parameterName=UseStyxShading}
  StaticSwitchParameter(True=EyeAdaptationInverse(LightValueInput=@n73).EyeAdaptationInverse, False=@n73) {defaultValue=true, group=Styx, parameterName=IsStyxRed}
  VertexNormalWS()
  VertexNormalWS()
 ]}
