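# Material function that overrides the Specular attribute for Styx water shading, passing attributes through when UseStyxShading is off.
# Specular scales between low-quality and full parameters by the ReflectionQuality collection value, stays at the low value on Low quality, and drops to zero on back faces via TwoSidedSign.
# With UseStyxTransition, TransitionMask blends from incoming to Styx specular, except translucent materials, which keep the incoming value.
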MaterialFunction /FortWater/Materials/Functions/MF_Water_Styx_SpecularOverride
graph {nodes [
  Constant()
  n1 = FunctionInput() {inputType=FunctionInput_MaterialAttributes}
  n2 = GetMaterialAttributes(None=@n1) {attributeGetTypes [#1]}
  n3 = ScalarParameter() {defaultValue=0.04, group=Styx, parameterName=Styx_Specular_LowQuality}
  n4 = ScalarParameter() {defaultValue=0.1, group=Styx, parameterName=Styx_Specular}
  n5 = CollectionParameter() {collection=&FortniteMaterialParameters, parameterName=ReflectionQuality}
  n6 = Saturate(None=@n5)
  n7 = LinearInterpolate(B=QualitySwitch(Default=LinearInterpolate(A=@n3, B=@n4, Alpha=@n6), Low=@n3), Alpha=Saturate(None=TwoSidedSign()))
  n8 = FunctionInput() {inputName=TransitionMask, inputType=FunctionInput_Scalar, sortPriority=3}
  n9 = StaticSwitchParameter(True=StaticSwitch(True=@n2.Specular, False=LinearInterpolate(A=@n2.Specular, B=@n7, Alpha=@n8), Value=NamedRerouteUsage()), False=@n7) {group=Styx, parameterName=UseStyxTransition}
  FunctionOutput(None=StaticSwitchParameter(
    True=SetMaterialAttributes!(MaterialAttributes=@n2.MaterialAttributes, Specular=@n9) {attributeSetTypes [#1]}
    False=@n1
   ) {parameterName=UseStyxShading})
  NamedRerouteDeclaration(None=FunctionInput(Preview=StaticBool()) {bUsePreviewValueAsDefault=true, inputName=Is Translucent Mat, inputType=FunctionInput_StaticBool, sortPriority=555}) {name=Is Translucent Mateiral, nodeColor=<rgba 0 1 0.1797 1>}
 ]}
