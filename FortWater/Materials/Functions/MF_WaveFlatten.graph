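# Water material function that varies wave normal strength across the world by flattening an input normal with a large-scale noise mask.
# World XY position, excluding shader offsets, is divided by a scale to sample a turbulence texture with the shared world-group wrap sampler.
# The red channel is boosted, clamped and used to blend flatness from zero to a macro maximum, which drives FlattenNormal.
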MaterialFunction /FortWater/Materials/Functions/MF_WaveFlatten
graph {nodes [
  n1 = ScalarParameter() {defaultValue=0.8, group=MacroFlatness, parameterName=FlattenessFromMacro}
  n2 = WorldPosition() {worldPositionShaderOffset=WPT_ExcludeAllShaderOffsets}
  n3 = TextureSampleParameter2D(UVs=Divide(
    A=ComponentMask(None=@n2.XYZ) {g=true, r=true}
    B=ScalarParameter() {defaultValue=8000, group=MacroFlatness, parameterName=FlatnessTextureScale}
   )) {
   group=MacroFlatness
   parameterName=FlatnessMacroTexture
   samplerSource=SSM_Wrap_WorldGroupSettings
   texture=/FortWater/Textures/T_Turbulence_Seamless
  }
  n4 = Clamp(None=Multiply(A=@n3.R, B=ScalarParameter() {defaultValue=7, group=MacroFlatness, parameterName=FlatnessTexturePremultiply}))
  FunctionOutput(None=MaterialFunctionCall(
    Normal=FunctionInput() {bUsePreviewValueAsDefault=true, inputName=Normal, previewValue=<0 0 1 1>}
    Flatness=LinearInterpolate(B=@n1, Alpha=@n4)
   ) {materialFunction=&FlattenNormal}.Result)
 ]}
