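# MaskOutWaves is a FortWater material function that suppresses waves on a water surface according to a scalar Mask input.
# It passes the Normal through FlattenNormal with one minus Mask as flatness, and multiplies the world position offset (WPO) by Mask.
# A mask of zero flattens both shading and displacement, while the default of one leaves the waves intact.
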MaterialFunction /FortWater/Materials/Functions/MaskOutWaves
graph {nodes [
  n1 = FunctionInput() {bUsePreviewValueAsDefault=true, inputName=Mask, inputType=FunctionInput_Scalar, sortPriority=2, previewValue=<1 0 0 1>}
  FunctionOutput(None=MaterialFunctionCall(Normal=FunctionInput() {inputName=Normal}, Flatness=OneMinus(None=@n1)) {materialFunction=&FlattenNormal}.Result) {outputName=Normal}
  FunctionOutput(None=Multiply(A=FunctionInput() {inputName=WPO, sortPriority=1}, B=@n1)) {outputName=WPO, sortPriority=1}
 ]}
