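# Material function that layers shore and ocean foam onto FortWater materials, outputting blended attributes, mask, color and scatter mask.
# Panned noise and sine-distorted world-space foam textures combine with depth, wave height and distance fades, using HeightLerp to shape foam.
# Sim input, far mask and Epic-only physics velocity field drive foam erasure, while an optional sphere mask retunes shallow-biome foam.
# Quality and shading-path switches give lower tiers, forward and mobile simpler foam and shading.
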MaterialFunction /FortWater/Materials/Functions/OceanShoreFoam_2
graph {nodes [
  n1 = GetMaterialAttributes(None=FunctionInput() {bUsePreviewValueAsDefault=true, inputName=Attributes, inputType=FunctionInput_MaterialAttributes}) {attributeGetTypes [#1]}
  ComponentMask(None=@n1.WorldPositionOffset) {b=true, desc=WPO}
  Constant3Vector() {desc="Doesn't do much and  its a Optimization win"}
  n2 = NamedRerouteUsage()
  n3 = Subtract(A=NamedRerouteUsage(), B=ScalarParameter() {group=Foam and Pondscum, parameterName=Foam Dark Color Cutoff})
  n4 = NamedRerouteUsage()
  n5 = StaticSwitchParameter(True=Saturate(None=Multiply(A=@n3, B=@n4)), False=Constant()) {defaultValue=true, parameterName=Use Ocean Foam}
  n6 = VectorParameter() {group=Water Shading Model, parameterName=FoamScattering, defaultValue=<rgba 0 0.1 0.2 5>}
  n7 = Divide(A=@n6.A) {constB=100}
  FunctionOutput(None=Multiply(A=Multiply(A=@n2, B=@n5), B=Multiply(A=@n6.RGB, B=@n7))) {outputName=Scatter Mask, sortPriority=3}
  n8 = GetMaterialAttributes(None=@n1.MaterialAttributes) {attributeGetTypes [
    #2
    #3
    #4
    #5
    #6
   ]}
  n9 = LinearInterpolate(A=@n8.BaseColor, B=NamedRerouteUsage(), Alpha=NamedRerouteUsage())
  n10 = LinearInterpolate(A=@n8.Opacity, B=ScalarParameter() {defaultValue=2, group=Foam and Pondscum, parameterName=Foam Opacity}, Alpha=NamedRerouteUsage())
  n11 = ShadingPathSwitch(Default=LinearInterpolate(A=@n8.Normal, B=NamedRerouteUsage(), Alpha=NamedRerouteUsage()), Mobile=@n8.Normal) {desc=Migrated from FeatureLevelSwitch}
  n12 = SetMaterialAttributes!(
   MaterialAttributes=@n8.MaterialAttributes
   BaseColor=@n9
   Specular=LinearInterpolate(A=@n8.Specular, B=ScalarParameter() {defaultValue=0.376, group=Foam and Pondscum, parameterName=Foam Specular}, Alpha=NamedRerouteUsage())
   Roughness=LinearInterpolate(A=@n8.Roughness, B=ScalarParameter() {defaultValue=0.6667, group=Foam and Pondscum, parameterName=Foam Roughness}, Alpha=NamedRerouteUsage())
   Opacity=@n10
   Normal=QualitySwitch(Default=@n11, Low=@n8.Normal, High=@n11, Medium=@n8.Normal, Epic=@n11)
  ) {attributeSetTypes [
    #2
    #3
    #4
    #5
    #6
   ]}
  n13 = SetMaterialAttributes!(MaterialAttributes=@n8.MaterialAttributes, BaseColor=@n9, Opacity=@n10) {attributeSetTypes [#2, #5]}
  n14 = QualitySwitch(Default=@n12, Low=@n13, High=@n12, Medium=@n13, Epic=@n12)
  FunctionOutput(None=ShadingPathSwitch(Default=@n14, Deferred=@n14, Forward=@n13, Mobile=@n13)) {outputName=Attributes}
  n15 = ComponentMask(None=WorldPosition().XYZ) {g=true, r=true}
  n16 = VectorParameter() {group=Biomes, parameterName=Shallow Foam Mask Location Radius, defaultValue=<rgba 81160 -98924 10000 1>}
  n17 = ComponentMask(None=@n16.RGB) {g=true, r=true}
  n18 = ScalarParameter() {defaultValue=0.85, group=Biomes, parameterName=Shallow Foam Mask Hardness}
  NamedRerouteDeclaration(None=OneMinus(None=SphereMask(A=@n15, B=@n17, Radius=@n16.B, Hardness=@n18) {desc=Custom Foam Mask})) {name=ShallowWaterMask, nodeColor=<rgba 0.1484 0 1 1>}
  n19 = Add(B=ComponentMask(None=WorldPosition() {worldPositionShaderOffset=WPT_ExcludeAllShaderOffsets}.XYZ) {g=true, r=true})
  n20 = Divide(A=@n19, B=ScalarParameter() {defaultValue=3000, group=Foam and Pondscum, parameterName=Foam2Scale})
  NamedRerouteDeclaration(None=@n20) {name=Simple WorldSpaceUVs, nodeColor=<rgba 0 0.1094 1 1>}
  n21 = TextureSample(UVs=NamedRerouteUsage()) {samplerType=SAMPLERTYPE_Normal, texture=/FortWater/Textures/T_OceanFoam_New_N}
  NamedRerouteDeclaration(None=@n21.RGB) {name=FoamNormal, nodeColor=<rgba 0.7109 1 0 1>}
  n22 = NamedRerouteUsage()
  n23 = NamedRerouteUsage()
  n24 = QualitySwitch(Default=@n22, Low=@n23, High=@n22, Medium=@n23, Epic=@n22)
  n25 = FunctionInput() {bUsePreviewValueAsDefault=true, inputName=Mask, inputType=FunctionInput_Scalar, sortPriority=2, previewValue=<1 0 0 1>}
  n26 = Multiply(
   A=StaticSwitchParameter(True=ShadingPathSwitch(Default=@n24, Deferred=@n24, Forward=@n23, Mobile=@n23), False=Constant()) {defaultValue=true, parameterName=Use Ocean Foam}
   B=NamedRerouteDeclaration(None=@n25) {name=InputMask, nodeColor=<rgba 0.5234 1 0 1>}
  )
  n27 = Multiply(
   A=@n26
   B=LinearInterpolate(A=ScalarParameter() {defaultValue=1, group=Foam and Pondscum, parameterName=Shallow Foam_Opacity}, Alpha=NamedRerouteUsage())
  )
  NamedRerouteDeclaration(None=FunctionOutput(None=StaticSwitchParameter(True=@n27, False=@n26) {parameterName=UseShallowFoamMask}) {outputName=Mask, sortPriority=1}) {name=Foam Mask With PlatformSwitch, nodeColor=<rgba 1 0 0.5156 1>}
  n28 = Divide(
   A=Subtract(A=NamedRerouteUsage(), B=ScalarParameter() {defaultValue=69.7077, parameterName=Depth Foam Offset_LOW})
   B=ScalarParameter() {defaultValue=300, group=Foam and Pondscum, parameterName=Depth for Foam Edge_LOW}
  )
  n29 = NamedRerouteUsage()
  n30 = Multiply(
   A=OneMinus(None=Saturate(None=@n28))
   B=Multiply(A=Subtract(A=@n29) {constB=0.5}, B=ScalarParameter() {defaultValue=0.75, group=Foam and Pondscum, parameterName=Foam2 TexHeight})
  )
  NamedRerouteDeclaration(None=@n30) {name=Final Foam GreyScale Mobile, nodeColor=<rgba 0 0.2734 1 1>}
  n31 = FunctionInput() {bUsePreviewValueAsDefault=true, inputName=WaveHeight, inputType=FunctionInput_Scalar, sortPriority=10}
  n32 = ScalarParameter() {defaultValue=40, group=Foam and Pondscum, parameterName=Foam2_Transition_MaxWaveHeight}
  n33 = Saturate(None=Divide(A=Add(A=@n31, B=@n32), B=Multiply(A=@n32) {constB=2}))
  n34 = ScalarParameter() {group=Foam and Pondscum, parameterName=Foam2_Transition_WaveInfluence}
  n35 = StaticSwitchParameter(
   True=LinearInterpolate(
    A=ScalarParameter() {defaultValue=0.05, group=Foam and Pondscum, parameterName=Shallow_Foam2_Transition_WaveInfluence}
    B=@n34
    Alpha=NamedRerouteUsage()
   )
   False=@n34
  ) {parameterName=UseShallowFoamMask}
  n36 = Max(A=Multiply(A=NamedRerouteUsage(), B=Multiply(A=@n33, B=@n35)), B=Multiply(A=NamedRerouteUsage(), B=NamedRerouteUsage()))
  n37 = ScalarParameter() {defaultValue=0.362, group=Foam and Pondscum, parameterName=Foam2_LargeScaleMaskSubtract}
  n38 = LinearInterpolate(
   A=ScalarParameter() {defaultValue=0.362, group=Foam and Pondscum, parameterName=Shallow_Foam_LargeScaleMaskSubtract}
   B=@n37
   Alpha=NamedRerouteUsage()
  )
  n39 = Subtract(
   A=Add(A=Saturate(None=@n36), B=NamedRerouteUsage())
   B=Multiply(A=NamedRerouteUsage(), B=StaticSwitchParameter(True=@n38, False=@n37) {parameterName=UseShallowFoamMask})
  )
  n40 = NamedRerouteUsage()
  n41 = ShadingPathSwitch(Default=QualitySwitch(Default=@n39, Epic=Add(A=@n39, B=@n40)), Mobile=@n39) {desc=Migrated from FeatureLevelSwitch}
  n42 = MaterialFunctionCall(
   Transition Phase=Min(A=Saturate(None=@n41), B=ScalarParameter() {defaultValue=0.5058, group=Foam and Pondscum, parameterName=Faom_Max})
   Height Texture=NamedRerouteUsage()
   Contrast=Constant()
  ) {materialFunction=&HeightLerp}
  n43 = Saturate(None=Subtract(A=@n42.Lerp Alpha No Contrast, B=NamedRerouteUsage()))
  NamedRerouteDeclaration(None=@n43) {name=Mask for Scattering, nodeColor=<rgba 0.7344 0 1 1>}
  n44 = PixelDepth()
  n45 = Divide(A=@n44.R, B=ScalarParameter() {defaultValue=20000, group=Foam and Pondscum, parameterName=Foam_Fade_Distance_1}) {constB=20000}
  NamedRerouteDeclaration(None=OneMinus(None=Saturate(None=@n45))) {name=FoamDistance Fade, nodeColor=<rgba 0 1 0.1797 1>}
  n46 = ScalarParameter() {defaultValue=252.0556, group=Foam and Pondscum, parameterName=Foam Depth Range}
  n47 = NamedRerouteUsage()
  n48 = LinearInterpolate(A=ScalarParameter() {defaultValue=252.0556, group=Foam and Pondscum, parameterName=Shallow FoamFromDepth Range}, B=@n46, Alpha=@n47)
  n49 = Saturate(None=Divide(A=NamedRerouteUsage(), B=StaticSwitchParameter(True=@n48, False=@n46) {parameterName=UseShallowFoamMask}))
  n50 = ScalarParameter() {defaultValue=0.576, group=Foam and Pondscum, parameterName=Foam From Depth amount}
  n51 = LinearInterpolate(A=ScalarParameter() {defaultValue=0.1, group=Foam and Pondscum, parameterName=Shallow FoamFromDepth amount}, B=@n50, Alpha=@n47)
  NamedRerouteDeclaration(None=Multiply(A=OneMinus(None=@n49), B=StaticSwitchParameter(True=@n51, False=@n50) {parameterName=UseShallowFoamMask})) {name=FoamAmountfromDepth, nodeColor=<rgba 1 0.7266 0 1>}
  n52 = ScalarParameter() {defaultValue=200000, group=Foam and Pondscum, parameterName=Foam_Fade_Distance_2}
  n53 = OneMinus(None=Saturate(None=Divide(A=@n44.R, B=@n52) {constB=50000}))
  NamedRerouteDeclaration(None=@n53) {name=FoamDistanceFade2, nodeColor=<rgba 0.2422 0 1 1>}
  n54 = ComponentMask(None=WorldPosition().XYZ) {g=true, r=true}
  n55 = ScalarParameter() {defaultValue=152949.6094, group=Water Shading Model, parameterName=Ocean RandomFoamMaskSize}
  n56 = ScalarParameter() {defaultValue=0.008, parameterName=RandomFoamMaskSpeed_1}
  n57 = Multiply(A=@n56) {constB=-1}
  n58 = TextureSample(UVs=Panner(Coordinate=Divide(A=@n54, B=@n55), Speed=AppendVector(A=@n56, B=@n57))) {samplerSource=SSM_Wrap_WorldGroupSettings, samplerType=SAMPLERTYPE_Masks, texture=/FortWater/Textures/OceanNoise__Mask}
  n59 = ComponentMask(None=WorldPosition().XYZ) {g=true, r=true}
  n60 = ScalarParameter() {defaultValue=24413.5781, group=Water Shading Model, parameterName=Ocean RandomFoamMaskSize2}
  n61 = TextureSample(UVs=Panner(Coordinate=Divide(A=@n59, B=@n60), Speed=AppendVector(A=@n57, B=@n56))) {samplerSource=SSM_Wrap_WorldGroupSettings, samplerType=SAMPLERTYPE_Masks, texture=&OceanNoise__Mask}
  NamedRerouteDeclaration(None=Multiply(A=Multiply(A=@n58.R, B=@n61.G), B=ScalarParameter() {defaultValue=1.4, parameterName=RandomFoamAddMult})) {name=Far Distance random Foam, nodeColor=<rgba 0 1 0.0625 1>}
  n62 = Divide(A=@n19, B=ScalarParameter() {defaultValue=20000, group=Foam and Pondscum, parameterName=Foam2Scale2})
  NamedRerouteDeclaration(None=@n62) {name=SimpleWorldSpaceUVsLargeScale, nodeColor=<rgba 0.1484 1 0 1>}
  n63 = Multiply(A=Time(), B=ScalarParameter() {defaultValue=0.05, group=Foam and Pondscum, parameterName=FoamWaveSpeed})
  n64 = Add(A=@n63, B=MaterialFunctionCall() {materialFunction=&Pi}.Result) {constB=0.5}
  n65 = Multiply(
   A=AppendVector(A=Sine(None=@n63), B=Sine(None=@n64))
   B=ScalarParameter() {defaultValue=0.051, group=Foam and Pondscum, parameterName=FoamWaveSineMotion}
  )
  n66 = TextureSample(UVs=NamedRerouteUsage()) {samplerSource=SSM_Wrap_WorldGroupSettings, samplerType=SAMPLERTYPE_Masks, texture=/FortWater/Textures/T_OceanFoam_New_Mask}
  n67 = Multiply(A=ScalarParameter() {defaultValue=1, group=Foam and Pondscum, parameterName=FoamWave_Distortion}, B=@n66.B)
  n68 = Multiply(A=@n65) {constB=-1}
  n69 = OneMinus(None=@n67)
  n70 = NamedRerouteUsage()
  n71 = Add(A=Add(A=Multiply(A=@n65, B=@n67), B=Multiply(A=@n68, B=@n69)), B=@n70)
  n72 = QualitySwitch(Default=@n71, Low=@n70, High=@n71, Medium=@n70, Epic=@n71)
  n73 = ShadingPathSwitch(Default=@n72, Deferred=@n72, Forward=@n70, Mobile=@n70)
  n74 = TextureSample(UVs=NamedRerouteDeclaration(None=@n73) {name=FoamUvs, nodeColor=<rgba 0 1 0.7656 1>}) {samplerSource=SSM_Wrap_WorldGroupSettings, samplerType=SAMPLERTYPE_Masks, texture=&T_OceanFoam_New_Mask}
  NamedRerouteDeclaration(None=@n74.R) {name=FoamTexture, nodeColor=<rgba 0 0.8125 1 1>}
  NamedRerouteDeclaration(None=@n66.G) {name=FoamTextureLargeScaleMask, nodeColor=<rgba 1 0 0.4219 1>}
  n75 = WorldPosition()
  n76 = WorldPosition() {worldPositionShaderOffset=WPT_ExcludeAllShaderOffsets}
  n77 = Multiply(
   A=Subtract(A=ComponentMask(None=@n75.XYZ) {b=true}, B=ComponentMask(None=@n76.XYZ) {b=true})
   B=ScalarParameter() {defaultValue=0.02, group=Foam and Pondscum, parameterName=Foam WPO Wave Influence}
  )
  NamedRerouteDeclaration(None=@n77) {name=WPOFoamInfluence, nodeColor=<rgba 0 0.6719 1 1>}
  NamedRerouteDeclaration(None=FunctionInput() {bUsePreviewValueAsDefault=true, inputName=Depth, inputType=FunctionInput_Scalar, sortPriority=3}) {name=Depth, nodeColor=<rgba 0 0.4375 1 1>}
  n78 = Constant()
  n79 = WorldPosition()
  n80 = Multiply(
   A=MaterialFunctionCall(Vector 3=SamplePhysicsVectorField!(World Position=@n79.XYZ) {fieldTarget=Vector_LinearVelocity}) {materialFunction=/Engine/Functions/Engine_MaterialFunctions02/Utility/VectorLength}.V3 Length
   B=ScalarParameter() {defaultValue=2, group=Foam and Pondscum, parameterName=Water Physics Force Foam Erasure}
  )
  n81 = ScalarParameter() {defaultValue=0.15, group=Foam and Pondscum, parameterName=Far Foam2 Remove}
  n82 = LinearInterpolate(A=ScalarParameter() {group=Foam and Pondscum, parameterName=Shallow Far Foam2 Remove}, B=@n81, Alpha=NamedRerouteUsage())
  n83 = Add(
   A=Add(
    A=FunctionInput() {bUsePreviewValueAsDefault=true, inputName=Sim Erasure, inputType=FunctionInput_Scalar, sortPriority=1}
    B=QualitySwitch(Default=@n78, Epic=@n80)
   )
   B=Multiply(
    A=FunctionInput() {bUsePreviewValueAsDefault=true, inputName=Far Mask, inputType=FunctionInput_Scalar, sortPriority=2}
    B=StaticSwitchParameter(True=@n82, False=@n81) {parameterName=UseShallowFoamMask}
   )
  )
  NamedRerouteDeclaration(None=@n83) {name=PhysicsForceFoamRemove, nodeColor=<rgba 0.9219 1 0 1>}
  n84 = VectorParameter() {group=Foam and Pondscum, parameterName=Foam2 Color 1, defaultValue=<rgba 0.319 0.9302 1 0.4>}
  n85 = VectorParameter() {group=Foam and Pondscum, parameterName=Foam2 Color 2, defaultValue=<rgba 0.817 0.9083 1 0.38>}
  n86 = Multiply(A=@n85.RGB, B=@n85.RGB)
  n87 = LinearInterpolate(A=Multiply(A=@n84.RGB, B=@n84.RGB), B=@n86, Alpha=Multiply(A=NamedRerouteUsage(), B=NamedRerouteUsage()))
  n88 = QualitySwitch(Default=@n87, Low=@n86, High=@n87, Medium=@n86)
  NamedRerouteDeclaration(None=FunctionOutput(None=ShadingPathSwitch(Default=@n88, Deferred=@n88, Forward=@n86, Mobile=@n86)) {outputName=Color, sortPriority=2}) {name=Final Foam Color, nodeColor=<rgba 0.7109 1 0 1>}
  NamedRerouteDeclaration(None=@n43) {name=Final Foam GreyScale, nodeColor=<rgba 1 0.8203 0 1>}
  Power(Base=@n42.Lerp Alpha No Contrast, Exp=ScalarParameter() {defaultValue=1, parameterName=FoamContrast})
  Reroute(None=@n20)
  Reroute(None=@n62)
  Reroute(None=@n77)
  Reroute(None=@n83)
  Reroute(None=@n21.RGB)
  Reroute(None=@n53)
  Reroute(None=@n73)
  Reroute(None=@n74.R)
  Reroute(None=@n30)
 ]}
