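# Material function for FortWater that rotates a world-space 2D normal into a view-aligned View Space output, by its name for water refraction.
# It projects the view forward axis onto world XY, normalizes it, and dots the normal with that heading and its perpendicular.
# The heading is also output for optional Custom UVs, whose input can replace it to save instructions.
# Refraction shift and camera-direction sign branches reach no output.
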MaterialFunction /FortWater/Materials/Functions/Transform_Refraction_Normals
graph {nodes [
  n1 = FunctionInput() {inputName=World Normal, inputType=FunctionInput_Vector2}
  n2 = ComponentMask(None=Transform(None=Constant3Vector() {constant=<rgba 0 0 1 0>}) {transformSourceType=TRANSFORMSOURCE_View}) {g=true, r=true}
  n3 = Normalize(VectorInput=@n2)
  n4 = FunctionInput(Preview=@n3) {
   bUsePreviewValueAsDefault=true
   description="When using Custom UVs, plug in the UV channel that you used as the Custom UV chanel for the Custom UV output pin on this node."
   inputName=Custom UV Input
   inputType=FunctionInput_Vector2
   sortPriority=1
  }
  n5 = MaterialFunctionCall(Float2=@n4) {materialFunction=&BreakOutFloat2Components}
  n6 = Multiply(A=@n5.R) {constB=-1}
  n7 = AppendVector(
   A=DotProduct(A=@n1, B=MaterialFunctionCall(X=@n5.G, Y=@n6) {materialFunction=&MakeFloat2}.Result)
   B=DotProduct(A=@n1, B=@n4)
  )
  n8 = CameraVectorWS()
  n9 = Multiply(
   A=MaterialFunctionCall(Vector 2=ComponentMask(None=@n8) {g=true, r=true}) {materialFunction=/Engine/Functions/Engine_MaterialFunctions02/Utility/VectorLength}.V2 Length
   B=ScalarParameter() {parameterName=Refraction shift}
  )
  Add(A=@n7, B=Multiply(A=@n9, B=Constant2Vector() {g=1}))
  ComponentMask(None=@n9) {g=true, r=true}
  FunctionOutput(None=@n7) {outputName=View Space}
  FunctionOutput(None=@n3) {description=Optional output to use Custom UVs to optimize a few instructions., outputName=To Custom UVs, sortPriority=1}
  n10 = ComponentMask(None=MaterialFunctionCall() {materialFunction=&CameraDirectionVector}.Result) {b=true}
  Multiply(
   A=@n2
   B=MaterialFunctionCall!(In=Multiply(A=@n10) {constB=-1}) {materialFunction=/Engine/Functions/Engine_MaterialFunctions02/Utility/Sign}.Result
  )
 ]}
