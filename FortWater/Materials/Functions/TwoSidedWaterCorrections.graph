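# FortWater material function that adjusts water material attributes so the surface renders correctly from both above and below.
# Back faces replace Specular with a Fresnel blend using a sign-flipped normal, and use their own refraction strength, reduced near the camera.
# Refraction also scales with capped input depth relative to pixel depth and field of view.
# A camera depth fade scales Specular, Opacity and refraction offset, and is also output as NearMask.
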MaterialFunction /FortWater/Materials/Functions/TwoSidedWaterCorrections
graph {nodes [
  n1 = GetMaterialAttributes(None=FunctionInput() {inputName=Attributes, inputType=FunctionInput_MaterialAttributes}) {attributeGetTypes [#1, #2, #3]}
  n2 = Constant() {r=1}
  n3 = MaterialFunctionCall(X=@n2, Y=@n2, Z=TwoSidedSign()) {materialFunction=&MakeFloat3}
  n4 = LinearInterpolate(
   A=ScalarParameter() {parameterName=Bottom Side Edge Specular}
   B=ScalarParameter() {defaultValue=2, parameterName=Bottom Side Specular}
   Alpha=Fresnel(Normal=Multiply(A=@n1.Normal, B=@n3.Result)) {baseReflectFraction=0, exponent=6}
  )
  n5 = Saturate(None=TwoSidedSign())
  n6 = MaterialFunctionCall(
   Fade Length=ScalarParameter() {defaultValue=2, parameterName=Near Fade length}
   Fade Offset=ScalarParameter() {defaultValue=10, parameterName=Near Fade offset}
  ) {materialFunction=&CameraDepthFade}
  n7 = FunctionInput() {bUsePreviewValueAsDefault=true, inputName=Depth, inputType=FunctionInput_Scalar, sortPriority=1, previewValue=<100 0 0 1>}
  n8 = Min(A=@n7) {constB=50}
  n9 = ViewProperty() {property=MEVP_TanHalfFieldOfView}
  n10 = Divide(
   A=Divide(A=PixelDepth().R, B=ComponentMask(None=@n9.Property) {r=true})
   B=ScalarParameter() {defaultValue=8, parameterName=Refraction Depth Scale}
  ) {constB=4}
  n11 = ScalarParameter() {defaultValue=0.05, parameterName=BottomRefraction}
  n12 = Saturate(None=Divide(A=PixelDepth().R, B=ScalarParameter() {defaultValue=1000, parameterName=Bottom Refraction Reduction Radius}))
  n13 = TwoSidedSign()
  n14 = Multiply(
   A=Saturate(None=Divide(A=@n8, B=@n10) {constA=100})
   B=LinearInterpolate(A=Multiply(A=@n11, B=@n12), B=ScalarParameter() {defaultValue=0.1, parameterName=Refraction}, Alpha=Saturate(None=@n13))
  )
  n15 = SetMaterialAttributes!(
   MaterialAttributes=@n1.MaterialAttributes
   Specular=Multiply(A=LinearInterpolate(A=@n4, B=@n1.Specular, Alpha=@n5), B=@n6.Result)
   Normal=@n1.MaterialAttributes
   Opacity=Multiply(A=@n1.Opacity, B=@n6.Result)
   Refraction=Add(B=Multiply(A=@n6.Result, B=@n14)) {constA=1}
  ) {attributeSetTypes [
    #1
    #2
    #3
    #4
   ]}
  FunctionOutput(None=@n15)
  FunctionOutput(None=@n6.Result) {outputName=NearMask, sortPriority=1}
 ]}
