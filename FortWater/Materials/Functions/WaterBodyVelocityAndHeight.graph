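# Water material function that samples a landscape-mapped velocity and height texture to drive flow, depth and surface height.
# Velocity and depth become four exponential masks in one vectorized Custom exp for waves, flow and mobile depth color.
# Masked velocity drives a noise-timed flowmap for normals, WPO and foam, fading flow and enlarging tiling by Far Mask.
# Sampled height also outputs WPO snapping vertices to water level, gated by EnableTerrainMapping.
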MaterialFunction /FortWater/Materials/Functions/WaterBodyVelocityAndHeight
graph {nodes [
  n1 = CollectionParameter() {collection=/Landmass/Landscape/BlueprintBrushes/MPC/MPC_Landscape, desc=Handled in Finalizatin Shader, parameterName=LandscapeZLocation}
  n2 = WorldPosition() {worldPositionShaderOffset=WPT_ExcludeAllShaderOffsets}
  n3 = Subtract(
   A=ComponentMask(None=FunctionInput(Preview=@n2.XYZ) {bUsePreviewValueAsDefault=true, inputName=Position, sortPriority=3}) {g=true, r=true}
   B=ComponentMask(None=CollectionParameter() {collection=&MPC_Landscape, parameterName=LandscapeLocation}) {g=true, r=true}
  )
  n4 = ComponentMask(None=CollectionParameter() {collection=&MPC_Landscape, parameterName=RTWorldSize}) {g=true, r=true}
  n5 = TextureSample!(
   UVs=Saturate(None=Divide(A=@n3, B=@n4))
   Tex=FunctionInput(Preview=TextureObjectParameter() {group=Technical, parameterName=WaterVelocityAndHeight, texture=&Black}) {bUsePreviewValueAsDefault=true, inputName=Texture, inputType=FunctionInput_Texture2D, sortPriority=1}
   Level=ScalarParameter() {group=Technical, parameterName=WaterVelocityAndHeightMipLevel}
   Apply View MipBias
  ) {mipValueMode=TMVM_MipLevel, samplerSource=SSM_Wrap_WorldGroupSettings, texture=&Black}
  n6 = Floor(None=FunctionOutput(None=@n5.B) {outputName=Water Z Height, sortPriority=4})
  Add(
   A=@n1
   B=Multiply(A=CollectionParameter() {collection=&MPC_Landscape, parameterName=LandscapeZScale}, B=@n6)
  )
  n7 = FunctionOutput(None=Multiply(A=OneMinus(None=@n5.A)) {constB=1000}) {outputName=Water Depth, sortPriority=3}
  n8 = ScalarParameter() {defaultValue=32, group=Flowmaps, parameterName=Min Shore Flow}
  n9 = ComponentMask(None=@n5.RGB) {g=true, r=true}
  n10 = MaterialFunctionCall(Vector 2=@n9) {materialFunction=/Engine/Functions/Engine_MaterialFunctions02/Utility/VectorLength}
  n11 = CollectionParameter() {collection=/Water/Materials/MPC/MPC_Water, parameterName=TargetWaveDepth}
  n12 = ScalarParameter() {defaultValue=256, group=Flowmaps, parameterName=Target Depth for Max Velocity Flow}
  n13 = ScalarParameter() {defaultValue=256, group=Mobile, parameterName=Target Depth for Full Opacity}
  n14 = ScalarParameter() {defaultValue=32, group=Flowmaps, parameterName=Target Velocity for Flowmap Alpha}
  MaterialExpressionDivide_5 = Divide(
   A=MaterialFunctionCall(X=@n7, Y=Add(A=@n8, B=@n7), Z=@n7, A=@n10.V2 Length) {materialFunction=&MakeFloat4}.Result
   B=Divide(A=MaterialFunctionCall(X=@n11, Y=@n12, Z=@n13, A=@n14) {materialFunction=&MakeFloat4}.Result)
  )
  n15 = Custom!(x=@MaterialExpressionDivide_5) {
   code="return 1-exp(-x);"
   description=exp
   outputType=CMOT_Float4
   inputs [{inputName=x, input {expression=&self:MaterialExpressionDivide_5}}]
  }
  AppendVector(A=@n15, B=@n10.V2 Length)
  Constant() {r=0.5}
  Constant()
  Constant()
  n16 = TextureObjectParameter() {
   group=Flowmaps
   parameterName=FlowmapWaveNormalHeight
   samplerType=SAMPLERTYPE_LinearColor
   texture=/Game/Global/Textures/Default/Tileable_Patterns/T_Water_TilingNormal_With_Height_02_Softened
  }
  n17 = ScalarParameter() {defaultValue=800, group=Flowmaps, parameterName=FlowmapWaveNormalScale}
  n18 = ScalarParameter() {defaultValue=0.5, group=Flowmaps, parameterName=Phase Fade Time}
  n19 = Divide(A=Multiply(A=ScalarParameter() {defaultValue=1, parameterName=Final Flow Mult}, B=Divide(A=@n9, B=@n17)), B=@n18)
  n20 = FunctionOutput(None=@n15) {
   description="Outputs 4 exponential masks using a single vectorized Exp operation. R=Wave Masked based on Depth, G=Flowmap Mask based on Depth, B=Coloration Mask based on Depth, A=Flow Mask based on velocity"
   outputName=Exponential Depth Masks
   sortPriority=5
  }
  n21 = MaterialFunctionCall(Float4=@n20) {materialFunction=&BreakOutFloat4Components}
  n22 = FunctionInput() {bUsePreviewValueAsDefault=true, inputName=Far Mask, inputType=FunctionInput_Scalar}
  n23 = WorldPosition() {worldPositionShaderOffset=WPT_ExcludeAllShaderOffsets}
  n24 = ScalarParameter() {defaultValue=1378, group=Flowmaps, parameterName=FlowmapWaveNormalScale}
  n25 = ScalarParameter() {defaultValue=8000, group=Distance Blending, parameterName=FarScale}
  n26 = TextureSample!(
   UVs=MaterialFunctionCall!(Scale=ScalarParameter() {defaultValue=2000, parameterName=Flowmap Breakup Scale}) {materialFunction=/Engine/Functions/Engine_MaterialFunctions01/Coordinates/WorldPosition-XY}.Result
   Tex
   Apply View MipBias
  ) {
   samplerSource=SSM_Wrap_WorldGroupSettings
   samplerType=SAMPLERTYPE_LinearGrayscale
   texture=/Engine/Functions/Engine_MaterialFunctions02/ExampleContent/Textures/LowResBlurredNoise
  }
  n27 = Multiply(A=ScalarParameter() {defaultValue=0.0001, group=Flowmaps, parameterName=FLow}, B=@n26.R)
  n28 = MaterialFunctionCall!(
   Diffuse=@n16
   Normal Map
   "Flow Vector Map (see tooltip)"=LinearInterpolate(A=Multiply(A=@n19, B=@n21.G), Alpha=@n22) {constB=0}
   UV texture 2 offset and Scale=VectorParameter() {group=Flowmaps, parameterName=Flowmap UV2 Offset and Scale, defaultValue=<rgba 0.618 1.618 1.382 0>}.RGB
   "Time (see tooltip)"=MaterialFunctionCall(Speed=Add(A=@n18, B=@n27) {constB=0}) {materialFunction=&TimeWithSpeedVariable}.Time
   UVs=Divide(A=ComponentMask(None=@n23.XYZ) {g=true, r=true}, B=LinearInterpolate(A=@n24, B=@n25, Alpha=@n22))
  ) {materialFunction=&/WaterFlowMap}
  n29 = ConstantBiasScale(None=@n28.Diffuse) {bias=-0.5, scale=2}
  n30 = ComponentMask(None=@n29) {g=true, r=true}
  n31 = ScalarParameter() {defaultValue=0.1, group=Flowmaps, parameterName=WavelNormalIntensity}
  FunctionOutput(None=DeriveNormalZ(InXY=Multiply(A=@n30, B=@n31))) {outputName=Flowmap Normals}
  n32 = Constant2Vector()
  n33 = Multiply(A=ComponentMask(None=@n29) {b=true}, B=ScalarParameter() {defaultValue=8, group=Flowmaps, parameterName=Wave Z Displacement})
  FunctionOutput(None=Multiply(A=AppendVector(A=@n32, B=@n33), B=@n21.G)) {outputName=Flowmap WPO, sortPriority=1}
  n34 = VectorParameter() {group=Mobile, parameterName=Mobile Shallow Color, defaultValue=<rgba 0 0.6 1 0.2>}
  n35 = VectorParameter() {group=Mobile, parameterName=Mobile Deep Color, defaultValue=<rgba 0 0.4 1 0.03>}
  FunctionOutput(None=LinearInterpolate(A=Multiply(A=@n34.RGB, B=@n34.RGB), B=Multiply(A=@n35.RGB, B=@n35.RGB), Alpha=ComponentMask(None=@n21.B) {b=true})) {
   desc=Outputs colors to use for depth coloring on opaque materials if needed on some platforms
   outputName=Mobile Depth Coloration
   sortPriority=-1
  }
  n36 = WorldPosition()
  n37 = Add(A=Subtract(A=@n6, B=ComponentMask(None=@n36.XYZ) {b=true}), B=ScalarParameter() {group=Technical, parameterName=EdgeWaterOffset})
  n38 = Constant3Vector() {constant=<rgba 0 0 1 0>}
  FunctionOutput(None=Multiply(A=Multiply(A=@n37, B=@n38), B=CollectionParameter() {collection=&MPC_Water, parameterName=EnableTerrainMapping})) {outputName=Water Height WPO, sortPriority=2}
  FunctionOutput(None=AppendVector(A=Multiply(A=@n9, B=@n21.G), B=@n10.V2 Length)) {outputName=VelocityAndLength}
  FunctionOutput(None=AppendVector(A=@n26.R, B=ComponentMask(None=@n28.Diffuse Alpha) {r=true})) {
   description="Outputs 4 exponential masks using a single vectorized Exp operation. R=Wave Masked based on Depth, G=Flowmap Mask based on Depth, B=Coloration Mask based on Depth, A=Flow Mask based on velocity"
   outputName=Phase Offset And Foam
   sortPriority=10
  }
  Multiply(A=@n30, B=ScalarParameter() {group=Flowmaps, parameterName=Wave XY Displacement})
  VertexInterpolator(VS=@n30)
 ]}
