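# Material function for Fortnite water that scales the Specular and Opacity of incoming material attributes by an opacity mask, with the mask fading out over distance.
# Pixel depth, shifted and divided by the Specular_DistanceFade offset and blend parameters, is saturated to drive a lerp from the mask to 1.
# Nearby pixels therefore receive the full mask, while distant pixels keep their original Specular and Opacity.
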MaterialFunction /FortWater/Materials/Functions/WaterOpacityMultiply
graph {nodes [
  n1 = GetMaterialAttributes(None=FunctionInput() {inputName=Attributes, inputType=FunctionInput_MaterialAttributes}) {attributeGetTypes [#1, #2]}
  n2 = FunctionInput() {bUsePreviewValueAsDefault=true, inputName=Opacity Mask, inputType=FunctionInput_Scalar, sortPriority=1, previewValue=<1 0 0 1>}
  n3 = Subtract(A=PixelDepth().R, B=ScalarParameter() {defaultValue=50000, group=Water Shading Model, parameterName=Specular_DistanceFade_Offset})
  n4 = ScalarParameter() {defaultValue=50000, group=Water Shading Model, parameterName=Specular_DistanceFade_Blend}
  n5 = LinearInterpolate(A=@n2, Alpha=Saturate(None=Divide(A=@n3, B=@n4)))
  FunctionOutput(None=SetMaterialAttributes!(MaterialAttributes=@n1.MaterialAttributes, Specular=Multiply(A=@n1.Specular, B=@n5), Opacity=Multiply(A=@n1.Opacity, B=@n5)) {attributeSetTypes [#1, #2]})
 ]}
