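# Material function that gives particles Gerstner wave displacement and a matching normal, so they follow a water body's surface.
# Water body index and depth default to dynamic parameter 3's red and green channels, with position defaulting to offset-free world position.
# Depth-based attenuation scales the Fort_ComputeGerstnerWaves offset and flattens its normal.
# A static switch, on by default, keeps only the vertical offset by zeroing XY after the wave computation.
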MaterialFunction /FortWater/Materials/Functions/WaterParticleDisplacement
graph {nodes [
  n1 = Constant2Vector()
  n2 = DynamicParameter() {parameterIndex=3, defaultValue=<rgba 0 512 1 0>, paramNames [WaterBodyIndex, WaterDepth, WaveAttenuationFactor, Param4]}
  n3 = FunctionInput(Preview=@n2.WaterBodyIndex) {
   bUsePreviewValueAsDefault=true
   description="Default = DP3 - Red Channel"
   inputName=WaterBodyIndex
   inputType=FunctionInput_Scalar
   sortPriority=1
  }
  n4 = MaterialFunctionCall!(
   Z Offset Only
   Position=FunctionInput(Preview=WorldPosition() {worldPositionShaderOffset=WPT_ExcludeAllShaderOffsets}.XYZ) {bUsePreviewValueAsDefault=true, inputName=Position}
   WaterBodyIndex=@n3
   Normals use VS Interpolation
   All Normals VS
  ) {materialFunction=&/Fort_ComputeGerstnerWaves}
  n5 = FunctionInput(Preview=@n2.WaterDepth) {
   bUsePreviewValueAsDefault=true
   description="Default =  DP3 - Green Channel"
   inputName=Depth
   inputType=FunctionInput_Scalar
   sortPriority=2
  }
  n6 = MaterialFunctionCall!(WaterDepth=@n5, Water Body Index=@n3) {materialFunction=/Water/Materials/Functions/ComputeWavesAttenuationFactor}
  n7 = Multiply(A=@n4.WPO, B=@n6.WaveAttenuationFactor)
  n8 = ComponentMask(None=@n7) {b=true}
  n9 = StaticBool() {value=true}
  FunctionOutput(None=StaticSwitch(
    True=AppendVector(A=@n1, B=@n8)
    False=@n7
    Value=FunctionInput(Preview=@n9) {bUsePreviewValueAsDefault=true, inputName=Ignore XY Offsets, inputType=FunctionInput_StaticBool}
   )) {outputName=WPO, sortPriority=1}
  FunctionOutput(None=MaterialFunctionCall(Normal=@n4.Normal, Flatness=OneMinus(None=@n6.WaveAttenuationFactor)) {materialFunction=&FlattenNormal}.Result) {outputName=Normal}
  StaticBool() {desc=This method is disabled because it doesn't account for steepness}
 ]}
