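# Water material function that adds ripples driven by physics field forces to incoming material attributes.
# It samples the linear velocity physics field at world position, whose direction flows a world-XY-mapped wave normal texture through WaterFlowMap.
# An exponential saturation of velocity magnitude scales those normals before they are added to the input Normal, while scaled magnitude adds to Emissive, zero by default.
# Mobile shading passes attributes through unchanged.
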MaterialFunction /FortWater/Materials/Functions/WaterPhysicsForce
graph {nodes [
  n1 = FunctionInput() {inputName=Attributes, inputType=FunctionInput_MaterialAttributes}
  n2 = GetMaterialAttributes(None=@n1) {attributeGetTypes [#1, #2]}
  n3 = SamplePhysicsVectorField!(World Position=WorldPosition().XYZ) {fieldTarget=Vector_LinearVelocity}
  n4 = MaterialFunctionCall(Vector 3=@n3) {materialFunction=/Engine/Functions/Engine_MaterialFunctions02/Utility/VectorLength}
  n5 = Add(A=@n2.EmissiveColor, B=Multiply(A=@n4.V3 Length, B=ScalarParameter() {parameterName=Water Physics Force Multiplier}))
  MaterialExpressionMultiply_2 = Multiply(A=@n4.V3 Length, B=ScalarParameter() {defaultValue=16, parameterName=Physics Force Denity})
  n6 = Custom!(x=@MaterialExpressionMultiply_2) {
   code="return 1.0-exp(-x);"
   description=exp
   outputType=CMOT_Float1
   inputs [{inputName=x, input {expression=&self:MaterialExpressionMultiply_2}}]
  }
  n7 = TextureObjectParameter() {
   group=Flowmaps
   parameterName=FlowmapWaveNormalHeight
   samplerType=SAMPLERTYPE_LinearColor
   texture=/Game/Global/Textures/Default/Tileable_Patterns/T_Water_TilingNormal_With_Height_02_Softened
  }
  n8 = ComponentMask(None=MaterialFunctionCall(Vector=@n3) {materialFunction=&SafeNormalize}.Result) {g=true, r=true}
  n9 = ScalarParameter() {defaultValue=0.5, parameterName=Physics Force Flowmap Velocity}
  n10 = ScalarParameter() {defaultValue=2536, parameterName=Physics Force Ripple Scale}
  n11 = ConstantBiasScale(None=MaterialFunctionCall!(
    Diffuse=@n7
    Normal Map
    "Flow Vector Map (see tooltip)"=Multiply(A=@n8, B=@n9)
    UV texture 2 offset and Scale
    "Time (see tooltip)"
    UVs=MaterialFunctionCall!(Scale=@n10) {materialFunction=/Engine/Functions/Engine_MaterialFunctions01/Coordinates/WorldPosition-XY}.Result
   ) {materialFunction=&/WaterFlowMap}.Diffuse) {bias=-0.5, scale=2}
  n12 = Multiply(A=Multiply(A=@n6, B=ComponentMask(None=@n11) {g=true, r=true}), B=ScalarParameter() {defaultValue=0.5, parameterName=Physics Force Normal})
  n13 = Constant() {r=1}
  n14 = Add(A=@n2.Normal, B=Normalize(VectorInput=AppendVector(A=@n12, B=@n13)))
  FunctionOutput(None=ShadingPathSwitch(
    Default=SetMaterialAttributes!(MaterialAttributes=@n2.MaterialAttributes, EmissiveColor=@n5, Normal=@n14) {attributeSetTypes [#1, #2]}
    Mobile=@n1
   ))
 ]}
