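# Material function for pool water that outputs an animated, flow-mapped normal from a tiling water texture, driven by vertex color.
# The red vertex channel, remapped to a signed value and copied into X and Y, becomes a local-space flow vector transformed and scaled by velocity and water-depth fade.
# A flow map function animates world-XY UVs using water time.
# The result is intensity-scaled before Z is rederived.
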MaterialFunction /FortWater/Materials/Functions/WaterPoolVertexColorFlowMotion
graph {nodes [
  n1 = TextureObject() {samplerType=SAMPLERTYPE_LinearColor, texture=/Game/Global/Textures/Default/Tileable_Patterns/T_Water_TilingNormal_With_Height_02_Softened}
  n2 = VertexColor()
  n3 = AppendVector(A=@n2.R, B=@n2.R)
  n4 = AppendVector(A=Multiply(A=Subtract(A=@n3) {constB=0.5}) {constB=2}, B=Constant())
  n5 = Divide(
   A=Subtract(A=SceneDepthWithoutWater!(UVs).R, B=PixelDepth().R)
   B=ScalarParameter() {group=VertexColorFlowmotion, parameterName=VCFM_DepthVelocityFade_Distance}
  )
  n6 = LinearInterpolate(
   A=ScalarParameter() {group=VertexColorFlowmotion, parameterName=VCFM_DepthVelocityFade_LowestScaling}
   Alpha=Power(Base=Saturate(None=@n5), Exp=ScalarParameter() {group=VertexColorFlowmotion, parameterName=VCFM_DepthVelocityFade_Exp})
  )
  n7 = ScalarParameter() {group=VertexColorFlowmotion, parameterName=VCFM_VelocityScaling}
  n8 = ScalarParameter() {defaultValue=1500, group=VertexColorFlowmotion, parameterName=VCFM_PondscumTiling}
  n9 = Divide(A=Multiply(A=Transform(None=@n4) {transformSourceType=TRANSFORMSOURCE_Local}, B=Multiply(A=@n6, B=@n7)), B=@n8)
  n10 = MaterialFunctionCall(
   Time=MaterialFunctionCall() {materialFunction=/Water/Materials/Functions/GetWaterTime}.WaterTime
   Speed=ScalarParameter() {defaultValue=0.7, group=VertexColorFlowmotion, parameterName=VCFM_FadePhase}
  ) {materialFunction=&TimeWithSpeedVariable}
  n11 = Divide(A=ComponentMask(None=WorldPosition() {worldPositionShaderOffset=WPT_ExcludeAllShaderOffsets}.XYZ) {g=true, r=true}, B=@n8)
  n12 = ComponentMask(None=ConstantBiasScale(None=MaterialFunctionCall!(Diffuse=@n1, Normal Map, "Flow Vector Map (see tooltip)"=@n9, UV texture 2 offset and Scale, "Time (see tooltip)"=@n10.Time, UVs=@n11) {materialFunction=&/WaterFlowMap}.Diffuse) {bias=-0.5, scale=2}) {g=true, r=true}
  n13 = ScalarParameter() {defaultValue=0.3, group=VertexColorFlowmotion, parameterName=VCFM_WavelNormalIntensity}
  FunctionOutput(None=DeriveNormalZ(InXY=Multiply(A=@n12, B=@n13)))
 ]}
