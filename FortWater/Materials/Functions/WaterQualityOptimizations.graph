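# Material function for water materials that simplifies Specular, Roughness and Refraction on lower quality levels and cheaper shading paths.
# It splits the incoming material attributes, runs those three through nested Quality and Shading Path switches, then writes them back.
# Deferred at High or default quality keeps the authored values.
# Low, Medium, Forward and Mobile substitute low-detail scalar parameters for Specular and Roughness, and a constant one for Refraction.
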MaterialFunction /FortWater/Materials/Functions/WaterQualityOptimizations
graph {nodes [
  n1 = GetMaterialAttributes(None=FunctionInput() {inputName=Attributes, inputType=FunctionInput_MaterialAttributes}) {attributeGetTypes [
    #1
    #2
    #3
    #4
    #5
    #6
    #7
   ]}
  n2 = ScalarParameter() {defaultValue=0.4, group=Water Shading Model, parameterName=Specular_LowDetail}
  n3 = QualitySwitch(Default=@n1.Specular, Low=@n2, High=@n1.Specular, Medium=@n2)
  n4 = ScalarParameter() {defaultValue=0.1, group=Water Shading Model, parameterName=RoughnessLowDetail}
  n5 = QualitySwitch(Default=@n1.Roughness, Low=@n4, High=@n1.Roughness, Medium=@n4)
  n6 = Constant() {r=1}
  n7 = QualitySwitch(Default=@n1.Refraction, Low=@n6, High=@n1.Refraction, Medium=@n6)
  n8 = SetMaterialAttributes!(
   MaterialAttributes=@n1.MaterialAttributes
   BaseColor=@n1.MaterialAttributes
   Specular=ShadingPathSwitch(Default=@n3, Deferred=@n3, Forward=@n2, Mobile=@n2)
   Roughness=ShadingPathSwitch(Default=@n5, Deferred=@n5, Forward=@n4, Mobile=@n4)
   Opacity=@n1.MaterialAttributes
   Normal=@n1.MaterialAttributes
   WorldPositionOffset=@n1.MaterialAttributes
   Refraction=ShadingPathSwitch(Default=@n7, Deferred=@n7, Forward=@n6, Mobile=@n6)
  ) {attributeSetTypes [
    #1
    #2
    #3
    #4
    #5
    #6
    #7
   ]}
  FunctionOutput(None=@n8)
 ]}
